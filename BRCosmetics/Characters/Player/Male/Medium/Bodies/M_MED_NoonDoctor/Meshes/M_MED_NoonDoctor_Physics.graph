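# Physics asset for the M_MED_NoonDoctor male body cosmetic, covering its hood, rear coat flaps, strap tie and chained clip.
# Tiny base capsules root short capsule chains with tight angular limits; the clip chain branches off the first clip bone.
# Side flap bases lock angularly to limited deform-clavicle bodies.
# Deform-clavicle, hood-squish and strap-tie bases link to upper-spine bodies with free angular motion, while other bases have no parent constraint.
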PhysicsAsset /BRCosmetics/Characters/Player/Male/Medium/Bodies/M_MED_NoonDoctor/Meshes/M_MED_NoonDoctor_Physics
bodies {
 clavicle_l [{shapeType=Capsule, length=15.9315, radius=7.6122, shapeName=clavicle_l_capsule, center=<10.3383 0 0>, rotation=<rot 90 0 0>}]
 clavicle_r [{shapeType=Capsule, length=15.9315, radius=7.6122, shapeName=clavicle_r_capsule, center=<-10.3383 0 0>, rotation=<rot 90 180 0>}]
 deform_clavicle_l [{shapeType=Capsule, length=10, radius=8, shapeName=deform_clavicle_l_capsule, center=<9.3092 -5.4 0>, rotation=<rot 0 -90 0>}]
 deform_clavicle_l_base [{shapeType=Capsule, length=0.505, radius=0.505, shapeName=deform_clavicle_l_base_capsule, rotation=<rot 0 -90 0>}]
 deform_clavicle_r [{shapeType=Capsule, length=10, radius=8, shapeName=deform_clavicle_r_capsule, center=<-9.3092 5.4 0>, rotation=<rot 0 -90 -180>}]
 deform_clavicle_r_base [{shapeType=Capsule, length=0.505, radius=0.505, shapeName=deform_clavicle_r_base_capsule, rotation=<rot 0 -90 0>}]
 dyn_clip_01 [{
   shapeType=Capsule
   length=5.2441
   radius=2.6186
   shapeName=dyn_clip_01_capsule
   center=<3.907 1.468 2.3354>
   rotation=<rot 0 95.2061 -64.3729>
  }]
 dyn_clip_02 [{
   shapeType=Capsule
   length=4.8287
   radius=1.0517
   shapeName=dyn_clip_02_capsule
   center=<4.2464 1.455 -1.5376>
   rotation=<rot 0 -76.8968 -71.2305>
  }]
 dyn_clip_03 [{
   shapeType=Capsule
   length=3.0467
   radius=1.2538
   shapeName=dyn_clip_03_capsule
   center=<3.02 1.8401 -0.9205>
   rotation=<rot -15.7062 -90 29.1687>
  }]
 dyn_clip_base [{shapeType=Capsule, length=0.505, radius=0.505, shapeName=dyn_clip_base_capsule, rotation=<rot 0 -90 0>}]
 dyn_clip_chain_01 [{
   shapeType=Capsule
   length=1.1377
   radius=1.1377
   shapeName=dyn_clip_chain_01_capsule
   center=<1.0364 0.1105 0.0181>
   rotation=<rot 0.8188 5.8686 -90>
  }]
 dyn_clip_chain_02 [{shapeType=Capsule, length=2.3542, radius=1.6716, shapeName=dyn_clip_chain_02_capsule, rotation=<rot 90 0 0>}]
 dyn_coat_flap_rear_01 [{shapeType=Capsule, length=6, radius=4, shapeName=dyn_coat_flap_rear_01_capsule, center=<2 4.2396 0>}]
 dyn_coat_flap_rear_01_base [{shapeType=Capsule, length=0.505, radius=0.505, shapeName=dyn_coat_flap_rear_01_base_capsule, rotation=<rot 0 -90 0>}]
 dyn_coat_flap_rear_01_l_base [{shapeType=Capsule, length=0.505, radius=0.505, shapeName=dyn_coat_flap_rear_01_l_base_capsule, rotation=<rot 0 -90 0>}]
 dyn_coat_flap_rear_01_r_base [{shapeType=Capsule, length=0.505, radius=0.505, shapeName=dyn_coat_flap_rear_01_r_base_capsule, rotation=<rot 0 -90 0>}]
 dyn_coat_flap_rear_02 [{shapeType=Capsule, length=6, radius=4, shapeName=dyn_coat_flap_rear_02_capsule, center=<2 3.0011 0>}]
 dyn_coat_flap_rear_l_01 [{shapeType=Capsule, length=6, radius=4, shapeName=dyn_coat_flap_rear_l_01_capsule, center=<2 4.2396 0>}]
 dyn_coat_flap_rear_l_02 [{shapeType=Capsule, length=6, radius=4, shapeName=dyn_coat_flap_rear_l_02_capsule, center=<2 3.0011 0>}]
 dyn_coat_flap_rear_r_01 [{shapeType=Capsule, length=6, radius=4, shapeName=dyn_coat_flap_rear_r_01_capsule, center=<-2 -4.2396 0>, rotation=<rot 0 0 -180>}]
 dyn_coat_flap_rear_r_02 [{shapeType=Capsule, length=6, radius=4, shapeName=dyn_coat_flap_rear_r_02_capsule, center=<-2 -3.0011 0>, rotation=<rot 0 0 -180>}]
 dyn_coat_strap_tie [{shapeType=Capsule, length=10.4425, radius=13.9101, shapeName=dyn_coat_strap_tie_capsule, center=<22.6549 0 0>, rotation=<rot 0 -90 0>}]
 dyn_coat_strap_tie_base [{shapeType=Capsule, length=0.505, radius=0.505, shapeName=dyn_coat_strap_tie_base_capsule, rotation=<rot 0 -90 0>}]
 dyn_hood_front_l_01 [{shapeType=Capsule, length=6, radius=4, shapeName=dyn_hood_front_l_01_capsule, center=<2 2.0386 0>}]
 dyn_hood_front_l_02 [{shapeType=Capsule, length=6, radius=4, shapeName=dyn_hood_front_l_02_capsule, center=<2 2.6849 0>}]
 dyn_hood_front_l_base [{shapeType=Capsule, length=0.505, radius=0.505, shapeName=dyn_hood_front_l_base_capsule, rotation=<rot 0 -90 0>}]
 dyn_hood_front_r_01 [{shapeType=Capsule, length=5.9742, radius=4.0258, shapeName=dyn_hood_front_r_01_capsule, center=<-2 -2.0386 0>, rotation=<rot 0 0 -180>}]
 dyn_hood_front_r_02 [{shapeType=Capsule, length=5.9858, radius=4.0142, shapeName=dyn_hood_front_r_02_capsule, center=<-2 -2.6849 0>, rotation=<rot 0 0 -180>}]
 dyn_hood_front_r_base [{shapeType=Capsule, length=0.505, radius=0.505, shapeName=dyn_hood_front_r_base_capsule, rotation=<rot 0 -90 0>}]
 dyn_hood_rear_01 [{shapeType=Capsule, length=6, radius=4, shapeName=dyn_hood_rear_01_capsule, center=<2 0 0>}]
 dyn_hood_rear_02 [{shapeType=Capsule, length=6, radius=4, shapeName=dyn_hood_rear_02_capsule, center=<2 0.6934 0>}]
 dyn_hood_rear_03 [{shapeType=Capsule, length=6, radius=4, shapeName=dyn_hood_rear_03_capsule, center=<2 0.8563 0>}]
 dyn_hood_rear_base [{shapeType=Capsule, length=0.505, radius=0.505, shapeName=dyn_hood_rear_base_capsule, rotation=<rot 0 -90 0>}]
 dyn_hood_rear_l_01 [{shapeType=Capsule, length=6, radius=4, shapeName=dyn_hood_rear_l_01_capsule, center=<2 0 0>}]
 dyn_hood_rear_l_02 [{shapeType=Capsule, length=6, radius=4, shapeName=dyn_hood_rear_l_02_capsule, center=<2 0 0>}]
 dyn_hood_rear_l_base [{shapeType=Capsule, length=0.505, radius=0.505, shapeName=dyn_hood_rear_l_base_capsule, rotation=<rot 0 -90 0>}]
 dyn_hood_rear_r_01 [{shapeType=Capsule, length=6, radius=4, shapeName=dyn_hood_rear_r_01_capsule}]
 dyn_hood_rear_r_02 [{shapeType=Capsule, length=6, radius=4, shapeName=dyn_hood_rear_r_02_capsule}]
 dyn_hood_rear_r_base [{shapeType=Capsule, length=0.505, radius=0.505, shapeName=dyn_hood_rear_r_base_capsule, rotation=<rot 0 -90 0>}]
 dyn_hood_rear_squish [{
   shapeType=Capsule
   length=9.8258
   radius=5.0892
   shapeName=dyn_hood_rear_squish_capsule
   center=<-0.8353 -0.8064 0.3969>
   rotation=<rot -1.3133 -90 3.4338>
  }]
 dyn_hood_rear_squish_base [{shapeType=Capsule, length=0.505, radius=0.505, shapeName=dyn_hood_rear_squish_base_capsule, rotation=<rot 0 -90 0>}]
 dyn_hood_side_l_01 [{shapeType=Capsule, length=7.3438, radius=2.6562, shapeName=dyn_hood_side_l_01_capsule, center=<2 1.2378 0>}]
 dyn_hood_side_l_02 [{shapeType=Capsule, length=7.3438, radius=2.6562, shapeName=dyn_hood_side_l_02_capsule, center=<2 1.6997 0>}]
 dyn_hood_side_l_base [{shapeType=Capsule, length=0.505, radius=0.505, shapeName=dyn_hood_side_l_base_capsule, rotation=<rot 0 -90 0>}]
 dyn_hood_side_r_01 [{shapeType=Capsule, length=6.7188, radius=3.2812, shapeName=dyn_hood_side_r_01_capsule, center=<-2 -1.2378 0>, rotation=<rot 0 0 -180>}]
 dyn_hood_side_r_02 [{shapeType=Capsule, length=6.7188, radius=3.2812, shapeName=dyn_hood_side_r_02_capsule, center=<-2 -1.6997 0>, rotation=<rot 0 0 -180>}]
 dyn_hood_side_r_base [{shapeType=Capsule, length=0.505, radius=0.505, shapeName=dyn_hood_side_r_base_capsule, rotation=<rot 0 -90 0>}]
 spine_04 [{shapeType=Capsule, length=13.9066, radius=12.5305, rotation=<rot 10 -90 0>}]
 spine_05 [{shapeType=Capsule, length=8.452, radius=11.548, center=<0 -1.062 0>}]
}
constraints [
 {bone1Name=dyn_hood_front_l_01, bone2Name=dyn_hood_front_l_base, swing1LimitDegrees=12, swing2LimitDegrees=12, twistLimitDegrees=12}
 {bone1Name=dyn_hood_front_l_02, bone2Name=dyn_hood_front_l_01, swing1LimitDegrees=12, swing2LimitDegrees=12, twistLimitDegrees=12}
 {bone1Name=dyn_hood_front_r_01, bone2Name=dyn_hood_front_r_base, swing1LimitDegrees=12, swing2LimitDegrees=12, twistLimitDegrees=12}
 {bone1Name=dyn_hood_front_r_02, bone2Name=dyn_hood_front_r_01, swing1LimitDegrees=12, swing2LimitDegrees=12, twistLimitDegrees=12}
 {bone1Name=dyn_hood_rear_01, bone2Name=dyn_hood_rear_base, swing1LimitDegrees=12, swing2LimitDegrees=12, twistLimitDegrees=12}
 {bone1Name=dyn_hood_rear_02, bone2Name=dyn_hood_rear_01, swing1LimitDegrees=12, swing2LimitDegrees=12, twistLimitDegrees=12}
 {bone1Name=dyn_hood_rear_03, bone2Name=dyn_hood_rear_02, swing1LimitDegrees=12, swing2LimitDegrees=12, twistLimitDegrees=12}
 {bone1Name=dyn_hood_side_r_02, bone2Name=dyn_hood_side_r_01, swing1LimitDegrees=12, swing2LimitDegrees=12, twistLimitDegrees=12}
 {bone1Name=dyn_hood_rear_l_01, bone2Name=dyn_hood_rear_l_base, swing1LimitDegrees=12, swing2LimitDegrees=12, twistLimitDegrees=12}
 {bone1Name=dyn_hood_rear_l_02, bone2Name=dyn_hood_rear_l_01, swing1LimitDegrees=12, swing2LimitDegrees=12, twistLimitDegrees=12}
 {bone1Name=dyn_hood_rear_r_01, bone2Name=dyn_hood_rear_r_base, swing1LimitDegrees=12, swing2LimitDegrees=12, twistLimitDegrees=12}
 {bone1Name=dyn_hood_rear_r_02, bone2Name=dyn_hood_rear_r_01, swing1LimitDegrees=12, swing2LimitDegrees=12, twistLimitDegrees=12}
 {bone1Name=dyn_hood_side_l_01, bone2Name=dyn_hood_side_l_base, swing1LimitDegrees=12, swing2LimitDegrees=12, twistLimitDegrees=12}
 {bone1Name=dyn_hood_side_l_02, bone2Name=dyn_hood_side_l_01, swing1LimitDegrees=12, swing2LimitDegrees=12, twistLimitDegrees=12}
 {bone1Name=dyn_hood_side_r_01, bone2Name=dyn_hood_side_r_base, swing1LimitDegrees=12, swing2LimitDegrees=12, twistLimitDegrees=12}
 {bone1Name=dyn_coat_flap_rear_01, bone2Name=dyn_coat_flap_rear_01_base, swing1LimitDegrees=6, swing2LimitDegrees=8, twistLimitDegrees=8}
 {bone1Name=dyn_coat_flap_rear_02, bone2Name=dyn_coat_flap_rear_01, swing1LimitDegrees=6, swing2LimitDegrees=8, twistLimitDegrees=8}
 {bone1Name=dyn_coat_flap_rear_l_01, bone2Name=dyn_coat_flap_rear_01_l_base, swing1LimitDegrees=6, swing2LimitDegrees=8, twistLimitDegrees=8}
 {bone1Name=dyn_coat_flap_rear_l_02, bone2Name=dyn_coat_flap_rear_l_01, swing1LimitDegrees=6, swing2LimitDegrees=8, twistLimitDegrees=8}
 {bone1Name=dyn_coat_flap_rear_r_01, bone2Name=dyn_coat_flap_rear_01_r_base, swing1LimitDegrees=6, swing2LimitDegrees=8, twistLimitDegrees=8}
 {bone1Name=dyn_coat_flap_rear_r_02, bone2Name=dyn_coat_flap_rear_r_01, swing1LimitDegrees=6, swing2LimitDegrees=8, twistLimitDegrees=8}
 {bone1Name=dyn_clip_01, bone2Name=dyn_clip_base, swing1LimitDegrees=16, swing2LimitDegrees=8, twistLimitDegrees=16}
 {bone1Name=dyn_clip_02, bone2Name=dyn_clip_01, swing1LimitDegrees=8, swing2LimitDegrees=8, twistLimitDegrees=8}
 {bone1Name=dyn_clip_03, bone2Name=dyn_clip_02, swing1LimitDegrees=8, swing2LimitDegrees=8, twistLimitDegrees=8}
 {bone1Name=dyn_clip_chain_01, bone2Name=dyn_clip_01, swing1LimitDegrees=16, swing2LimitDegrees=16, twistLimitDegrees=16}
 {bone1Name=dyn_clip_chain_02, bone2Name=dyn_clip_chain_01, swing1LimitDegrees=24, swing2LimitDegrees=24, twistLimitDegrees=24}
 {
  bone1Name=dyn_coat_flap_rear_01_l_base
  bone2Name=deform_clavicle_l
  swing1LimitDegrees=45
  swing1Motion=Locked
  swing2LimitDegrees=45
  swing2Motion=Locked
  twistLimitDegrees=45
  twistMotion=Locked
 }
 {bone1Name=deform_clavicle_l, bone2Name=deform_clavicle_l_base, swing1LimitDegrees=8, swing2LimitDegrees=8, twistLimitDegrees=8}
 {
  bone1Name=dyn_coat_flap_rear_01_r_base
  bone2Name=deform_clavicle_r
  swing1LimitDegrees=45
  swing1Motion=Locked
  swing2LimitDegrees=45
  swing2Motion=Locked
  twistLimitDegrees=45
  twistMotion=Locked
 }
 {bone1Name=deform_clavicle_r, bone2Name=deform_clavicle_r_base, swing1LimitDegrees=8, swing2LimitDegrees=8, twistLimitDegrees=8}
 {
  bone1Name=dyn_coat_strap_tie_base
  bone2Name=spine_04
  swing1LimitDegrees=45
  swing1Motion=Free
  swing2LimitDegrees=45
  swing2Motion=Free
  twistLimitDegrees=45
  twistMotion=Free
 }
 {bone1Name=dyn_coat_strap_tie, bone2Name=dyn_coat_strap_tie_base, swing1LimitDegrees=4, swing2LimitDegrees=4, twistLimitDegrees=4}
 {
  bone1Name=dyn_hood_rear_squish_base
  bone2Name=spine_05
  swing1LimitDegrees=45
  swing1Motion=Free
  swing2LimitDegrees=45
  swing2Motion=Free
  twistLimitDegrees=45
  twistMotion=Free
 }
 {bone1Name=dyn_hood_rear_squish, bone2Name=dyn_hood_rear_squish_base, swing1LimitDegrees=8, swing2LimitDegrees=8, twistLimitDegrees=8}
 {
  bone1Name=deform_clavicle_l_base
  bone2Name=spine_05
  swing1LimitDegrees=45
  swing1Motion=Free
  swing2LimitDegrees=45
  swing2Motion=Free
  twistLimitDegrees=45
  twistMotion=Free
 }
 {
  bone1Name=deform_clavicle_r_base
  bone2Name=spine_05
  swing1LimitDegrees=45
  swing1Motion=Free
  swing2LimitDegrees=45
  swing2Motion=Free
  twistLimitDegrees=45
  twistMotion=Free
 }
]
