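# Physics asset for the NoonDoctor male medium cosmetic poncho, with capsule bodies for its dynamic coat bones and the pelvis, spine, legs and feet.
# Six front and rear chains (center, left, right) hang from tiny base capsules constrained to spine_04.
# Front joints share uniform limits, while rear joints keep tight twist and widen one swing limit toward the hem.
# Free-angular cross constraints link each side chain's tip to its center tip.
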PhysicsAsset /BRCosmetics/Characters/Player/Male/Medium/Bodies/M_MED_NoonDoctor/Meshes/M_MED_NoonDoctor_Poncho_Physics
bodies {
 calf_l [{shapeType=Capsule, length=32.0675, radius=10.2478, center=<-17.7339 -2.204 2.395>, rotation=<rot 0 -92.6095 -88.1249>}]
 calf_r [{shapeType=Capsule, length=32.0675, radius=10.2478, center=<17.7339 2.204 -2.395>, rotation=<rot 0 -92.6095 91.8751>}]
 dyn_coat_front_01 [{shapeType=Capsule, length=12, radius=6, shapeName=dyn_coat_front_01_capsule, center=<0 2.4183 0>}]
 dyn_coat_front_01_base [{shapeType=Capsule, length=0.505, radius=0.505, shapeName=dyn_coat_front_01_base_capsule, rotation=<rot 0 -90 0>}]
 dyn_coat_front_02 [{shapeType=Capsule, length=12, radius=6, shapeName=dyn_coat_front_02_capsule, center=<0 3.9187 0>}]
 dyn_coat_front_03 [{shapeType=Capsule, length=12, radius=6, shapeName=dyn_coat_front_03_capsule, center=<0 2.2698 0>}]
 dyn_coat_front_04 [{shapeType=Capsule, length=5.125, radius=12.875, shapeName=dyn_coat_front_04_capsule, center=<2.8963 7.1772 0>}]
 dyn_coat_front_l_01 [{shapeType=Capsule, length=12, radius=6, shapeName=dyn_coat_front_l_01_capsule, center=<0 4.3705 0>}]
 dyn_coat_front_l_02 [{shapeType=Capsule, length=12, radius=6, shapeName=dyn_coat_front_l_02_capsule, center=<0 5.2329 0>}]
 dyn_coat_front_l_03 [{shapeType=Capsule, length=12, radius=6, shapeName=dyn_coat_front_l_03_capsule, center=<0 4.3607 0>}]
 dyn_coat_front_l_04 [{shapeType=Capsule, length=6.4062, radius=11.5938, shapeName=dyn_coat_front_l_04_capsule, center=<3.5467 9.7844 0>}]
 dyn_coat_front_l_base [{shapeType=Capsule, length=0.505, radius=0.505, shapeName=dyn_coat_front_l_base_capsule, rotation=<rot 0 -90 0>}]
 dyn_coat_front_r_01 [{shapeType=Capsule, length=12, radius=6, shapeName=dyn_coat_front_r_01_capsule, center=<0 -4.4023 0>}]
 dyn_coat_front_r_02 [{shapeType=Capsule, length=12, radius=6, shapeName=dyn_coat_front_r_02_capsule, center=<0 -5.395 0>}]
 dyn_coat_front_r_03 [{shapeType=Capsule, length=12, radius=6, shapeName=dyn_coat_front_r_03_capsule, center=<0 -4.711 0>}]
 dyn_coat_front_r_04 [{
   shapeType=Capsule
   length=6.4062
   radius=11.5938
   shapeName=dyn_coat_front_r_04_capsule
   center=<-3.5467 -10.3296 0>
   rotation=<rot 0 0 -180>
  }]
 dyn_coat_front_r_base [{shapeType=Capsule, length=0.505, radius=0.505, shapeName=dyn_coat_front_r_base_capsule, rotation=<rot 0 -90 0>}]
 dyn_coat_rear_01 [{shapeType=Capsule, length=12, radius=6, shapeName=dyn_coat_rear_01_capsule}]
 dyn_coat_rear_01_r [{shapeType=Capsule, length=12, radius=6, shapeName=dyn_coat_rear_01_r_capsule}]
 dyn_coat_rear_02 [{shapeType=Capsule, length=12, radius=6, shapeName=dyn_coat_rear_02_capsule}]
 dyn_coat_rear_02_r [{shapeType=Capsule, length=12, radius=6, shapeName=dyn_coat_rear_02_r_capsule}]
 dyn_coat_rear_03 [{shapeType=Capsule, length=11.3438, radius=6.6562, shapeName=dyn_coat_rear_03_capsule, center=<0 3.4596 0>}]
 dyn_coat_rear_03_r [{shapeType=Capsule, length=11.3438, radius=6.6562, shapeName=dyn_coat_rear_03_r_capsule, center=<0 -4.1322 0>}]
 dyn_coat_rear_04 [{shapeType=Capsule, length=11.4375, radius=6.5625, shapeName=dyn_coat_rear_04_capsule, center=<0 2.9804 0>}]
 dyn_coat_rear_04_r [{shapeType=Capsule, length=11.4375, radius=6.5625, shapeName=dyn_coat_rear_04_r_capsule, center=<0 -3.4572 0>}]
 dyn_coat_rear_05 [{shapeType=Capsule, length=10.9375, radius=7.0625, shapeName=dyn_coat_rear_05_capsule, center=<3.6968 2.8 0>}]
 dyn_coat_rear_05_r [{shapeType=Capsule, length=11.1875, radius=6.8125, shapeName=dyn_coat_rear_05_r_capsule, center=<-3.4932 -2.8 0>}]
 dyn_coat_rear_base [{shapeType=Capsule, length=0.505, radius=0.505, shapeName=dyn_coat_rear_base_capsule, rotation=<rot 0 -90 0>}]
 dyn_coat_rear_l_01 [{shapeType=Capsule, length=12, radius=6, shapeName=dyn_coat_rear_l_01_capsule}]
 dyn_coat_rear_l_02 [{shapeType=Capsule, length=12, radius=6, shapeName=dyn_coat_rear_l_02_capsule}]
 dyn_coat_rear_l_03 [{shapeType=Capsule, length=11.3438, radius=6.6562, shapeName=dyn_coat_rear_l_03_capsule, center=<0 3.9432 0>}]
 dyn_coat_rear_l_04 [{shapeType=Capsule, length=11.4375, radius=6.5625, shapeName=dyn_coat_rear_l_04_capsule, center=<0 3.6703 0>}]
 dyn_coat_rear_l_05 [{shapeType=Capsule, length=10.9375, radius=7.0625, shapeName=dyn_coat_rear_l_05_capsule, center=<3.6968 2.8578 0>}]
 dyn_coat_rear_l_base [{shapeType=Capsule, length=0.505, radius=0.505, shapeName=dyn_coat_rear_l_base_capsule, rotation=<rot 0 -90 0>}]
 dyn_coat_rear_r_base [{shapeType=Capsule, length=0.505, radius=0.505, shapeName=dyn_coat_rear_r_base_capsule, rotation=<rot 0 -90 0>}]
 foot_l [{shapeType=Capsule, length=10.9635, radius=11.6802, center=<-2.8687 6.6574 -0.2921>, rotation=<rot 0 11.9606 -89.8608>}]
 foot_r [{shapeType=Capsule, length=10.9635, radius=11.6802, center=<2.8687 -6.6574 0.2921>, rotation=<rot 0 11.9606 90.1392>}]
 pelvis [{shapeType=Capsule, length=25.9205, radius=11.7687, shapeName=pelvis_capsule, center=<-2.1967 -1.6609 0>, rotation=<rot 0 -90 0>}]
 spine_02 [{shapeType=Capsule, length=16.5938, radius=11.4418, shapeName=spine_02_capsule, center=<0 -1.1273 0>, rotation=<rot 0 -90 0>}]
 spine_04 [{
   shapeType=Capsule
   length=14.8288
   radius=12.2262
   shapeName=spine_04_capsule
   center=<0.0584 -2.9095 0.3836>
   rotation=<rot 0.6795 -90 -0.4514>
  }]
 thigh_l [{
   shapeType=Capsule
   length=36.5555
   radius=11.9112
   shapeName=thigh_l_capsule
   center=<-17.0055 -1.8595 -0.9655>
   rotation=<rot 0.7821 -90.0018 -82.2254>
  }]
 thigh_r [{
   shapeType=Capsule
   length=36.5555
   radius=11.9112
   shapeName=thigh_r_capsule
   center=<17.0055 1.8595 0.9655>
   rotation=<rot 0.7821 -90.0018 97.7746>
  }]
}
constraints [
 {bone1Name=dyn_coat_rear_01, bone2Name=dyn_coat_rear_base, swing1LimitDegrees=8, swing2LimitDegrees=8, twistLimitDegrees=4}
 {bone1Name=dyn_coat_rear_02, bone2Name=dyn_coat_rear_01, swing1LimitDegrees=16, swing2LimitDegrees=8, twistLimitDegrees=4}
 {bone1Name=dyn_coat_rear_03, bone2Name=dyn_coat_rear_02, swing1LimitDegrees=24, swing2LimitDegrees=8, twistLimitDegrees=4}
 {bone1Name=dyn_coat_rear_04, bone2Name=dyn_coat_rear_03, swing1LimitDegrees=26, swing2LimitDegrees=8, twistLimitDegrees=4}
 {bone1Name=dyn_coat_rear_l_01, bone2Name=dyn_coat_rear_l_base, swing1LimitDegrees=8, swing2LimitDegrees=8, twistLimitDegrees=4}
 {bone1Name=dyn_coat_rear_l_02, bone2Name=dyn_coat_rear_l_01, swing1LimitDegrees=16, swing2LimitDegrees=8, twistLimitDegrees=4}
 {bone1Name=dyn_coat_rear_l_03, bone2Name=dyn_coat_rear_l_02, swing1LimitDegrees=24, swing2LimitDegrees=8, twistLimitDegrees=4}
 {bone1Name=dyn_coat_rear_l_04, bone2Name=dyn_coat_rear_l_03, swing1LimitDegrees=26, swing2LimitDegrees=8, twistLimitDegrees=4}
 {bone1Name=dyn_coat_rear_l_05, bone2Name=dyn_coat_rear_l_04, swing1LimitDegrees=32, swing2LimitDegrees=8, twistLimitDegrees=4}
 {bone1Name=dyn_coat_rear_01_r, bone2Name=dyn_coat_rear_r_base, swing1LimitDegrees=8, swing2LimitDegrees=8, twistLimitDegrees=4}
 {bone1Name=dyn_coat_rear_02_r, bone2Name=dyn_coat_rear_01_r, swing1LimitDegrees=16, swing2LimitDegrees=8, twistLimitDegrees=4}
 {bone1Name=dyn_coat_rear_03_r, bone2Name=dyn_coat_rear_02_r, swing1LimitDegrees=24, swing2LimitDegrees=8, twistLimitDegrees=4}
 {bone1Name=dyn_coat_rear_04_r, bone2Name=dyn_coat_rear_03_r, swing1LimitDegrees=26, swing2LimitDegrees=8, twistLimitDegrees=4}
 {bone1Name=dyn_coat_front_01, bone2Name=dyn_coat_front_01_base, swing1LimitDegrees=28, swing2LimitDegrees=18, twistLimitDegrees=18}
 {bone1Name=dyn_coat_front_02, bone2Name=dyn_coat_front_01, swing1LimitDegrees=28, swing2LimitDegrees=18, twistLimitDegrees=18}
 {bone1Name=dyn_coat_front_03, bone2Name=dyn_coat_front_02, swing1LimitDegrees=28, swing2LimitDegrees=18, twistLimitDegrees=18}
 {bone1Name=dyn_coat_front_04, bone2Name=dyn_coat_front_03, swing1LimitDegrees=28, swing2LimitDegrees=18, twistLimitDegrees=18}
 {bone1Name=dyn_coat_front_l_01, bone2Name=dyn_coat_front_l_base, swing1LimitDegrees=28, swing2LimitDegrees=18, twistLimitDegrees=18}
 {bone1Name=dyn_coat_front_l_02, bone2Name=dyn_coat_front_l_01, swing1LimitDegrees=28, swing2LimitDegrees=18, twistLimitDegrees=18}
 {bone1Name=dyn_coat_front_l_03, bone2Name=dyn_coat_front_l_02, swing1LimitDegrees=28, swing2LimitDegrees=18, twistLimitDegrees=18}
 {bone1Name=dyn_coat_front_l_04, bone2Name=dyn_coat_front_l_03, swing1LimitDegrees=28, swing2LimitDegrees=18, twistLimitDegrees=18}
 {bone1Name=dyn_coat_front_r_01, bone2Name=dyn_coat_front_r_base, swing1LimitDegrees=28, swing2LimitDegrees=18, twistLimitDegrees=18}
 {bone1Name=dyn_coat_front_r_02, bone2Name=dyn_coat_front_r_01, swing1LimitDegrees=28, swing2LimitDegrees=18, twistLimitDegrees=18}
 {bone1Name=dyn_coat_front_r_03, bone2Name=dyn_coat_front_r_02, swing1LimitDegrees=28, swing2LimitDegrees=18, twistLimitDegrees=18}
 {bone1Name=dyn_coat_front_r_04, bone2Name=dyn_coat_front_r_03, swing1LimitDegrees=28, swing2LimitDegrees=18, twistLimitDegrees=18}
 {
  bone1Name=dyn_coat_front_01_base
  bone2Name=spine_04
  swing1LimitDegrees=45
  swing1Motion=Free
  swing2LimitDegrees=45
  swing2Motion=Free
  twistLimitDegrees=45
  twistMotion=Free
 }
 {
  bone1Name=dyn_coat_front_r_base
  bone2Name=spine_04
  swing1LimitDegrees=45
  swing1Motion=Free
  swing2LimitDegrees=45
  swing2Motion=Free
  twistLimitDegrees=45
  twistMotion=Free
 }
 {
  bone1Name=dyn_coat_front_l_base
  bone2Name=spine_04
  swing1LimitDegrees=45
  swing1Motion=Free
  swing2LimitDegrees=45
  swing2Motion=Free
  twistLimitDegrees=45
  twistMotion=Free
 }
 {
  bone1Name=dyn_coat_rear_l_base
  bone2Name=spine_04
  swing1LimitDegrees=45
  swing1Motion=Free
  swing2LimitDegrees=45
  swing2Motion=Free
  twistLimitDegrees=45
  twistMotion=Free
 }
 {
  bone1Name=dyn_coat_rear_base
  bone2Name=spine_04
  swing1LimitDegrees=45
  swing1Motion=Free
  swing2LimitDegrees=45
  swing2Motion=Free
  twistLimitDegrees=45
  twistMotion=Free
 }
 {
  bone1Name=dyn_coat_rear_r_base
  bone2Name=spine_04
  swing1LimitDegrees=45
  swing1Motion=Free
  swing2LimitDegrees=45
  swing2Motion=Free
  twistLimitDegrees=45
  twistMotion=Free
 }
 {
  bone1Name=thigh_l
  bone2Name=pelvis
  swing1LimitDegrees=45
  swing1Motion=Free
  swing2LimitDegrees=45
  swing2Motion=Free
  twistLimitDegrees=45
  twistMotion=Free
 }
 {
  bone1Name=calf_l
  bone2Name=pelvis
  swing1LimitDegrees=45
  swing1Motion=Free
  swing2LimitDegrees=45
  swing2Motion=Free
  twistLimitDegrees=45
  twistMotion=Free
 }
 {
  bone1Name=foot_l
  bone2Name=pelvis
  swing1LimitDegrees=45
  swing1Motion=Free
  swing2LimitDegrees=45
  swing2Motion=Free
  twistLimitDegrees=45
  twistMotion=Free
 }
 {
  bone1Name=thigh_r
  bone2Name=pelvis
  swing1LimitDegrees=45
  swing1Motion=Free
  swing2LimitDegrees=45
  swing2Motion=Free
  twistLimitDegrees=45
  twistMotion=Free
 }
 {
  bone1Name=calf_r
  bone2Name=pelvis
  swing1LimitDegrees=45
  swing1Motion=Free
  swing2LimitDegrees=45
  swing2Motion=Free
  twistLimitDegrees=45
  twistMotion=Free
 }
 {
  bone1Name=foot_r
  bone2Name=pelvis
  swing1LimitDegrees=45
  swing1Motion=Free
  swing2LimitDegrees=45
  swing2Motion=Free
  twistLimitDegrees=45
  twistMotion=Free
 }
 {
  bone1Name=spine_02
  bone2Name=pelvis
  swing1LimitDegrees=45
  swing1Motion=Free
  swing2LimitDegrees=45
  swing2Motion=Free
  twistLimitDegrees=45
  twistMotion=Free
 }
 {
  bone1Name=spine_04
  bone2Name=spine_02
  swing1LimitDegrees=45
  swing1Motion=Free
  swing2LimitDegrees=45
  swing2Motion=Free
  twistLimitDegrees=45
  twistMotion=Free
 }
 {
  bone1Name=dyn_coat_front_l_04
  bone2Name=dyn_coat_front_04
  swing1LimitDegrees=45
  swing1Motion=Free
  swing2LimitDegrees=45
  swing2Motion=Free
  twistLimitDegrees=45
  twistMotion=Free
 }
 {
  bone1Name=dyn_coat_front_r_04
  bone2Name=dyn_coat_front_04
  swing1LimitDegrees=45
  swing1Motion=Free
  swing2LimitDegrees=45
  swing2Motion=Free
  twistLimitDegrees=45
  twistMotion=Free
 }
 {bone1Name=dyn_coat_rear_05, bone2Name=dyn_coat_rear_04, swing1LimitDegrees=32, swing2LimitDegrees=8, twistLimitDegrees=4}
 {bone1Name=dyn_coat_rear_05_r, bone2Name=dyn_coat_rear_04_r, swing1LimitDegrees=32, swing2LimitDegrees=8, twistLimitDegrees=4}
 {
  bone1Name=dyn_coat_rear_05_r
  bone2Name=dyn_coat_rear_05
  swing1LimitDegrees=45
  swing1Motion=Free
  swing2LimitDegrees=45
  swing2Motion=Free
  twistLimitDegrees=45
  twistMotion=Free
 }
 {
  bone1Name=dyn_coat_rear_l_05
  bone2Name=dyn_coat_rear_05
  swing1LimitDegrees=45
  swing1Motion=Free
  swing2LimitDegrees=45
  swing2Motion=Free
  twistLimitDegrees=45
  twistMotion=Free
 }
]
