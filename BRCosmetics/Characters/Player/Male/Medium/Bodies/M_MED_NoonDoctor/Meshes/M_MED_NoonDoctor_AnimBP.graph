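# Character-part AnimBP for the NoonDoctor player body, adding secondary motion to its coat, shoulder flaps and tassels.
# The copied parent pose passes through Control Rig, then main RigidBody and AnimDynamics.
# A constraint-anchored poncho RigidBody sim, used when cloth is off, layers onto coat and flap branches before tassel dynamics.
# The event graph drives cloth, wind and gravity settings, suppressing cloth for first person or cape backblings.
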AnimBlueprint /BRCosmetics/Characters/Player/Male/Medium/Bodies/M_MED_NoonDoctor/Meshes/M_MED_NoonDoctor_AnimBP
target_skeleton=&/M_MED_NoonDoctor_Skeleton
graphs [
 AnimGraph {nodes [
   n1 = Layered blend per bone(BasePose=Use cached pose 'Post Main Dynamics'(), BlendPoses_0=Use cached pose 'Post Poncho Sim'()) {layerSetup [{branchFilters [{boneName=dyn_coat_base}, {boneName=dyn_shoulder_flap_r_base}, {boneName=dyn_shoulder_flap_l_base}]}]}
   n2 = "Constraint - dyn_coat_tassle_01_base (PPPPPP)"!(
    ComponentPose=Local To Component(LocalPose=@n1)
    Alpha=1
    ConstraintWeights_0=0.5261
    ConstraintWeights_1:float=0.1565
    ConstraintWeights_2:float=0.1336
    ConstraintWeights_3:float=0.0764
    ConstraintWeights_4:float=0.0748
    ConstraintWeights_5:float=0.0326
   ) -> (Pose) {
    boneToModify {boneName=dyn_coat_tassle_01_base}
    constraintSetup [
     {targetBone {boneName=dyn_coat_strap_tie}}
     {targetBone {boneName=spine_03}}
     {targetBone {boneName=dyn_coat_rear_01}}
     {targetBone {boneName=spine_04}}
     {targetBone {boneName=spine_02}}
     {targetBone {boneName=spine_01}}
    ]
    constraintWeights [1,1,1,1,1,1]
   }
   n3 = "Constraint - dyn_coat_tassle_connection_r (PPPPPP)"!(
    ComponentPose=@n2
    Alpha=1
    ConstraintWeights_0=0.5261
    ConstraintWeights_1:float=0.1565
    ConstraintWeights_2:float=0.1336
    ConstraintWeights_3:float=0.0764
    ConstraintWeights_4:float=0.0748
    ConstraintWeights_5:float=0.0326
   ) -> (Pose) {
    boneToModify {boneName=dyn_coat_tassle_connection_r}
    constraintSetup [
     {targetBone {boneName=dyn_coat_strap_tie}}
     {targetBone {boneName=spine_03}}
     {targetBone {boneName=dyn_coat_rear_01}}
     {targetBone {boneName=spine_04}}
     {targetBone {boneName=spine_02}}
     {targetBone {boneName=spine_01}}
    ]
    constraintWeights [1,1,1,1,1,1]
   }
   n4 = "Constraint - dyn_coat_tassle_connection_l (PPPPPP)"!(
    ComponentPose=@n3
    Alpha=1
    ConstraintWeights_0=0.5294
    ConstraintWeights_1:float=0.1565
    ConstraintWeights_2:float=0.1248
    ConstraintWeights_3:float=0.0772
    ConstraintWeights_4:float=0.0771
    ConstraintWeights_5:float=0.0349
   ) -> (Pose) {
    boneToModify {boneName=dyn_coat_tassle_connection_l}
    constraintSetup [
     {targetBone {boneName=dyn_coat_strap_tie}}
     {targetBone {boneName=spine_03}}
     {targetBone {boneName=dyn_coat_rear_01}}
     {targetBone {boneName=spine_02}}
     {targetBone {boneName=spine_04}}
     {targetBone {boneName=dyn_coat_rear_l_01}}
    ]
    constraintWeights [1,1,1,1,1,1]
   }
   n5 = "Transform (Modify) Bone - Bone: dyn_coat_tassle_01"(ComponentPose=@n4, Alpha=1, Translation="0,0,0", Rotation="0,25,0", Scale="1,1,1") {
    rotationMode=BMM_Additive
    rotationSpace=BCS_BoneSpace
    alphaScaleBiasClamp {bClampResult=true, bInterpResult=true, bMapRange=true, inRange {max=700}}
    boneToModify {boneName=dyn_coat_tassle_01}
    rotation {yaw=25}
   }
   n6 = Anim Dynamics - Bone: dyn_coat_tassle_01(ComponentPose=@n5, Alpha=1, ExternalForce="0,0,0") {
    angularDampingOverride=0.999
    angularSpringConstant=15000
    bAngularSpring=true
    bOverrideAngularDamping=true
    bUseGravityOverride=true
    gravityScale=0.35
    simulationSpace=RootRelative
    boundBone {boneName=dyn_coat_tassle_01}
    componentAppliedLinearAccClamp=<0 0 0>
    gravityOverride {z=-2000}
    physicsBodyDefinitions [{
      boundBone {boneName=dyn_coat_tassle_01}
      boxExtents=<20 30 30>
      constraintSetup {angularLimitsMax=<15 15 18>, angularLimitsMin=<-15 -15 -28>, angularTarget=<1 0 0>}
      localJointOffset=<4 0 0>
     }]
    simSpaceSettings {maxAngularAcceleration=8, maxAngularVelocity=1200, simSpaceAngularAlpha=1}
   }
   n7 = Anim Dynamics - Bone: dyn_coat_tassle_02(ComponentPose=@n6, Alpha=1, ExternalForce="0,0,0") {
    angularDampingOverride=0.98
    angularSpringConstant=12000
    bAngularSpring=true
    bOverrideAngularDamping=true
    bUseGravityOverride=true
    gravityScale=0.35
    simulationSpace=RootRelative
    boundBone {boneName=dyn_coat_tassle_02}
    componentAppliedLinearAccClamp=<1400 1400 1400>
    componentLinearAccScale=<0.2 0.2 0.2>
    componentLinearVelScale=<0.2 0.2 0.2>
    gravityOverride {z=-980}
    physicsBodyDefinitions [{
      boundBone {boneName=dyn_coat_tassle_02}
      boxExtents=<20 30 30>
      constraintSetup {angularLimitsMax=<24 24 4>, angularLimitsMin=<-24 -24 -4>, angularTarget=<1 0 0>}
      localJointOffset=<4 0 0>
     }]
    simSpaceSettings {maxAngularAcceleration=8, maxAngularVelocity=1200, simSpaceAngularAlpha=1}
   }
   n8 = Anim Dynamics - Bone: dyn_coat_tassle_03(ComponentPose=@n7, Alpha=1, ExternalForce="0,0,0") {
    angularDampingOverride=0.94
    angularSpringConstant=2000
    bAngularSpring=true
    bOverrideAngularDamping=true
    bUseGravityOverride=true
    gravityScale=0.35
    simulationSpace=RootRelative
    boundBone {boneName=dyn_coat_tassle_03}
    componentAppliedLinearAccClamp=<1400 1400 1400>
    componentLinearAccScale=<0.2 0.2 0.2>
    componentLinearVelScale=<0.02 0.2 0.2>
    gravityOverride {z=-980}
    physicsBodyDefinitions [{
      boundBone {boneName=dyn_coat_tassle_03}
      boxExtents=<20 30 30>
      constraintSetup {angularLimitsMax=<32 32 6>, angularLimitsMin=<-32 -32 -6>, angularTarget=<1 0 0>}
      localJointOffset=<3 0 0>
     }]
    simSpaceSettings {maxAngularAcceleration=8, maxAngularVelocity=1200, simSpaceAngularAlpha=1}
   }
   n9 = Anim Dynamics - Bone: dyn_coat_tassle_04(ComponentPose=@n8, Alpha=1, ExternalForce="0,0,0") {
    angularDampingOverride=0.94
    angularSpringConstant=2000
    bAngularSpring=true
    bOverrideAngularDamping=true
    bUseGravityOverride=true
    gravityScale=0.35
    simulationSpace=RootRelative
    boundBone {boneName=dyn_coat_tassle_04}
    componentAppliedLinearAccClamp=<1400 1400 1400>
    componentLinearAccScale=<0.2 0.2 0.2>
    componentLinearVelScale=<0.02 0.2 0.2>
    gravityOverride {z=-980}
    physicsBodyDefinitions [{
      boundBone {boneName=dyn_coat_tassle_04}
      boxExtents=<20 30 30>
      constraintSetup {angularLimitsMax=<30 30 6>, angularLimitsMin=<-30 -30 -6>, angularTarget=<1 0 0>}
      localJointOffset=<4 0 0>
     }]
    simSpaceSettings {maxAngularAcceleration=8, maxAngularVelocity=1200, simSpaceAngularAlpha=1}
   }
   Output Pose(Result=Component To Local(ComponentPose=@n9)) {layerGroup=None}
   n10 = Control Rig!(Source=Copy Pose From Mesh() {bUseAttachedParent=true}, Alpha, Is Cloth Enabled:bool=Get bClothEnabled()) -> (Pose) {controlRigAssetReference {controlRigAsset=&/M_MED_NoonDoctor_CtrlRig}}
   Save cached pose 'Post Deformation'(Pose=@n10) {cacheName=Post Deformation}
   n11 = "Constraint - dyn_coat_rear_l_base (PPPPPPPP)"!(
    ComponentPose=Local To Component(LocalPose=Use cached pose 'Post Main Dynamics'())
    Alpha=1
    ConstraintWeights_0=0.3739
    ConstraintWeights_1:float=0.2144
    ConstraintWeights_2:float=0.1724
    ConstraintWeights_3:float=0.0827
    ConstraintWeights_4:float=0.0816
    ConstraintWeights_5:float=0.0722
    ConstraintWeights_6:float=0.0019
    ConstraintWeights_7:float=0.0009
   ) -> (Pose) {
    boneToModify {boneName=dyn_coat_rear_l_base}
    constraintSetup [
     {targetBone {boneName=dyn_coat_strap_tie}}
     {targetBone {boneName=spine_03}}
     {targetBone {boneName=spine_02}}
     {targetBone {boneName=spine_04}}
     {targetBone {boneName=spine_01}}
     {targetBone {boneName=pelvis}}
     {targetBone {boneName=spine_05}}
     {targetBone {boneName=deform_clavicle_l}}
    ]
    constraintWeights [1,1,1,1,1,1,1,1]
   }
   n12 = "Constraint - dyn_coat_rear_base (PPPPP)"!(
    ComponentPose=@n11
    Alpha=1
    ConstraintWeights_0=0.3649
    ConstraintWeights_1:float=0.2997
    ConstraintWeights_2:float=0.1413
    ConstraintWeights_3:float=0.1068
    ConstraintWeights_4:float=0.0873
   ) -> (Pose) {
    boneToModify {boneName=dyn_coat_rear_base}
    constraintSetup [
     {targetBone {boneName=spine_03}}
     {targetBone {boneName=spine_02}}
     {targetBone {boneName=spine_01}}
     {targetBone {boneName=pelvis}}
     {targetBone {boneName=spine_04}}
    ]
    constraintWeights [1,1,1,1,1]
   }
   n13 = "Constraint - dyn_coat_rear_r_base (PPPPPPPP)"!(
    ComponentPose=@n12
    Alpha=1
    ConstraintWeights_0=0.3959
    ConstraintWeights_1:float=0.2071
    ConstraintWeights_2:float=0.1653
    ConstraintWeights_3:float=0.0819
    ConstraintWeights_4:float=0.0771
    ConstraintWeights_5:float=0.0674
    ConstraintWeights_6:float=0.0037
    ConstraintWeights_7:float=0.0017
   ) -> (Pose) {
    boneToModify {boneName=dyn_coat_rear_r_base}
    constraintSetup [
     {targetBone {boneName=dyn_coat_strap_tie}}
     {targetBone {boneName=spine_03}}
     {targetBone {boneName=spine_02}}
     {targetBone {boneName=spine_04}}
     {targetBone {boneName=spine_01}}
     {targetBone {boneName=pelvis}}
     {targetBone {boneName=deform_clavicle_r}}
     {targetBone {boneName=spine_05}}
    ]
    constraintWeights [1,1,1,1,1,1,1,1]
   }
   n14 = RigidBody!(
    ComponentPose=@n13
    Alpha
    OverrideWorldGravity:Vector="0,0,0"
    ExternalForce
    SimSpaceSettings:Sim Space Settings Structure=(WorldAlpha=1,VelocityScaleZ=0.25,DampingAlpha=1,MaxLinearVelocity=770,MaxAngularVelocity=2,MaxLinearAcceleration=3000,MaxAngularAcceleration=28,ExternalLinearDragV=(X=0.1,Y=0.1,Z=0.1),ExternalLinearVelocity=(X=0,Y=0,Z=0),ExternalAngularVelocity=(X=0,Y=0,Z=0))
   ) -> (Pose) {
    bClampLinearTranslationLimitToRefPose=true
    bOverrideWorldGravity=true
    overridePhysicsAsset=&/M_MED_NoonDoctor_Poncho_Physics
    simulationSpace=BaseBoneSpace
    baseBoneRef {boneName=root}
    componentAppliedLinearAccClamp=<0 0 0>
    simSpaceSettings {
     maxAngularAcceleration=28
     maxAngularVelocity=2
     maxLinearAcceleration=3000
     maxLinearVelocity=770
     velocityScaleZ=0.25
     worldAlpha=1
     externalLinearDragV=<0.1 0.1 0.1>
    }
   }
   n15 = "Constraint - dyn_coat_rear_l_01 (T)"(ComponentPose=@n14, Alpha=1, ConstraintWeights_0=1) {
    boneToModify {boneName=dyn_coat_rear_l_01}
    constraintSetup [{transformType=Translation, targetBone {boneName=dyn_coat_rear_l_base}}]
    constraintWeights [1]
   }
   n16 = "Constraint - dyn_coat_rear_01 (T)"(ComponentPose=@n15, Alpha=1, ConstraintWeights_0=1) {
    boneToModify {boneName=dyn_coat_rear_01}
    constraintSetup [{transformType=Translation, targetBone {boneName=dyn_coat_rear_base}}]
    constraintWeights [1]
   }
   n17 = "Constraint - dyn_coat_rear_01_r (T)"(ComponentPose=@n16, Alpha=1, ConstraintWeights_0=1) {
    boneToModify {boneName=dyn_coat_rear_01_r}
    constraintSetup [{transformType=Translation, targetBone {boneName=dyn_coat_rear_r_base}}]
    constraintWeights [1]
   }
   n18 = "Constraint - dyn_coat_front_01 (T)"(ComponentPose=@n17, Alpha=1, ConstraintWeights_0=1) {
    boneToModify {boneName=dyn_coat_front_01}
    constraintSetup [{transformType=Translation, targetBone {boneName=dyn_coat_front_01_base}}]
    constraintWeights [1]
   }
   n19 = "Constraint - dyn_coat_front_l_01 (T)"(ComponentPose=@n18, Alpha=1, ConstraintWeights_0=1) {
    boneToModify {boneName=dyn_coat_front_l_01}
    constraintSetup [{transformType=Translation, targetBone {boneName=dyn_coat_front_l_01}}]
    constraintWeights [1]
   }
   n20 = "Constraint - dyn_coat_front_r_01 (T)"(ComponentPose=@n19, Alpha=1, ConstraintWeights_0=1) {
    boneToModify {boneName=dyn_coat_front_r_01}
    constraintSetup [{transformType=Translation, targetBone {boneName=dyn_coat_front_r_base}}]
    constraintWeights [1]
   }
   n21 = Anim Dynamics - Bone: dyn_shoulder_flap_l_02(ComponentPose=@n20, Alpha=1, ExternalForce="0,0,0") {
    angularDampingOverride=0.99
    angularSpringConstant=5000
    bAngularSpring=true
    bOverrideAngularDamping=true
    bUseGravityOverride=true
    gravityScale=0.35
    simulationSpace=BoneRelative
    boundBone {boneName=dyn_shoulder_flap_l_02}
    componentAppliedLinearAccClamp=<0 0 0>
    gravityOverride {z=-980}
    physicsBodyDefinitions [{
      boundBone {boneName=dyn_shoulder_flap_l_02}
      boxExtents=<20 30 30>
      constraintSetup {angularLimitsMax=<2 2 15>, angularLimitsMin=<-2 -2 0>, angularTarget=<1 0 0>}
      localJointOffset=<4 0 0>
     }]
    relativeSpaceBone {boneName=upperarm_twist_01_l}
    simSpaceSettings {maxAngularAcceleration=64, maxAngularVelocity=16, simSpaceAngularAlpha=1}
   }
   n22 = Anim Dynamics - Bone: dyn_shoulder_flap_l_03(ComponentPose=@n21, Alpha=1, ExternalForce="0,0,0") {
    angularDampingOverride=0.99
    angularSpringConstant=5000
    bAngularSpring=true
    bOverrideAngularDamping=true
    bUseGravityOverride=true
    gravityScale=0.35
    simulationSpace=BoneRelative
    boundBone {boneName=dyn_shoulder_flap_l_03}
    componentAppliedLinearAccClamp=<0 0 0>
    gravityOverride {z=-980}
    physicsBodyDefinitions [{
      boundBone {boneName=dyn_shoulder_flap_l_03}
      boxExtents=<20 30 30>
      constraintSetup {angularLimitsMax=<2 2 15>, angularLimitsMin=<-2 -2 0>, angularTarget=<1 0 0>}
      localJointOffset=<4 0 0>
     }]
    relativeSpaceBone {boneName=upperarm_twist_01_l}
    simSpaceSettings {maxAngularAcceleration=64, maxAngularVelocity=16, simSpaceAngularAlpha=1}
   }
   n23 = Anim Dynamics - Bone: dyn_shoulder_flap_r_02(ComponentPose=@n22, Alpha=1, ExternalForce="0,0,0") {
    angularDampingOverride=0.99
    angularSpringConstant=5000
    bAngularSpring=true
    bOverrideAngularDamping=true
    bUseGravityOverride=true
    gravityScale=0.35
    simulationSpace=BoneRelative
    boundBone {boneName=dyn_shoulder_flap_r_02}
    componentAppliedLinearAccClamp=<0 0 0>
    gravityOverride {z=-980}
    physicsBodyDefinitions [{
      boundBone {boneName=dyn_shoulder_flap_r_02}
      boxExtents=<20 30 30>
      constraintSetup {angularLimitsMax=<2 2 15>, angularLimitsMin=<-2 -2 0>, angularTarget=<1 0 0>}
      localJointOffset=<-4 0 0>
     }]
    relativeSpaceBone {boneName=upperarm_twist_01_r}
    simSpaceSettings {maxAngularAcceleration=64, maxAngularVelocity=16, simSpaceAngularAlpha=1}
   }
   n24 = Anim Dynamics - Bone: dyn_shoulder_flap_r_03(ComponentPose=@n23, Alpha=1, ExternalForce="0,0,0") {
    angularDampingOverride=0.99
    angularSpringConstant=5000
    bAngularSpring=true
    bOverrideAngularDamping=true
    bUseGravityOverride=true
    gravityScale=0.35
    simulationSpace=BoneRelative
    boundBone {boneName=dyn_shoulder_flap_r_03}
    componentAppliedLinearAccClamp=<0 0 0>
    gravityOverride {z=-980}
    physicsBodyDefinitions [{
      boundBone {boneName=dyn_shoulder_flap_r_03}
      boxExtents=<20 30 30>
      constraintSetup {angularLimitsMax=<2 2 15>, angularLimitsMin=<-2 -2 0>, angularTarget=<1 0 0>}
      localJointOffset=<-4 0 0>
     }]
    relativeSpaceBone {boneName=upperarm_twist_01_r}
    simSpaceSettings {maxAngularAcceleration=64, maxAngularVelocity=16, simSpaceAngularAlpha=1}
   }
   n25 = "Transform (Modify) Bone - Bone: dyn_coat_front_01_base"!(ComponentPose=@n24, bAlphaBoolEnabled:bool=True, Translation="0,0,0", Rotation="0,10,0", Scale="1,1,1") -> (Pose) {
    alphaInputType=Bool
    rotationMode=BMM_Additive
    rotationSpace=BCS_BoneSpace
    alphaBoolBlend {blendInTime=0.2, blendOutTime=0.2}
    boneToModify {boneName=dyn_coat_front_01_base}
    rotation {yaw=10}
   }
   n26 = "Transform (Modify) Bone - Bone: dyn_coat_front_l_base"!(ComponentPose=@n25, bAlphaBoolEnabled:bool=True, Translation="0,0,0", Rotation="0,10,0", Scale="1,1,1") -> (Pose) {
    alphaInputType=Bool
    rotationMode=BMM_Additive
    rotationSpace=BCS_BoneSpace
    alphaBoolBlend {blendInTime=0.2, blendOutTime=0.2}
    boneToModify {boneName=dyn_coat_front_l_base}
    rotation {yaw=10}
   }
   n27 = "Transform (Modify) Bone - Bone: dyn_coat_front_r_base"!(ComponentPose=@n26, bAlphaBoolEnabled:bool=True, Translation="0,0,0", Rotation="0,10,0", Scale="1,1,1") -> (Pose) {
    alphaInputType=Bool
    rotationMode=BMM_Additive
    rotationSpace=BCS_BoneSpace
    alphaBoolBlend {blendInTime=0.2, blendOutTime=0.2}
    boneToModify {boneName=dyn_coat_front_r_base}
    rotation {yaw=10}
   }
   Save cached pose 'Post Poncho Sim'(Pose=Blend Poses by bool(
     BlendPose_0=Use cached pose 'Post Deformation'()
     BlendPose_1=Component To Local(ComponentPose=@n27)
     BlendTime_0=0
     BlendTime_1=0
     bActiveValue=Get bClothEnabled()
    )) {cacheName=Post Poncho Sim}
   n28 = RigidBody!(ComponentPose=Local To Component(LocalPose=Use cached pose 'Post Deformation'()), Alpha, OverrideWorldGravity="0,0,0", ExternalForce) -> (Pose) {
    bClampLinearTranslationLimitToRefPose=true
    bOverrideWorldGravity=true
    overridePhysicsAsset=&/M_MED_NoonDoctor_Physics
    simulationSpace=BaseBoneSpace
    baseBoneRef {boneName=spine_04}
    componentAppliedLinearAccClamp=<0 0 0>
    simSpaceSettings {
     dampingAlpha=0.5
     maxAngularAcceleration=28
     maxAngularVelocity=2
     maxLinearAcceleration=3000
     maxLinearVelocity=770
     velocityScaleZ=0.35
     worldAlpha=1
     externalLinearDragV=<0.1 0.1 0.1>
    }
   }
   n29 = Anim Dynamics - Bone: dyn_coat_shoulder_l_01(ComponentPose=@n28, Alpha=1, ExternalForce="0,0,0") {
    angularDampingOverride=1
    angularSpringConstant=20000
    bAngularSpring=true
    bOverrideAngularDamping=true
    bUseGravityOverride=true
    gravityScale=0.35
    simulationSpace=RootRelative
    boundBone {boneName=dyn_coat_shoulder_l_01}
    componentLinearAccScale=<1 1 1>
    componentLinearVelScale=<1 1 1>
    gravityOverride {z=-980}
    physicsBodyDefinitions [{
      boundBone {boneName=dyn_coat_shoulder_l_01}
      boxExtents=<20 30 30>
      constraintSetup {angularLimitsMax=<15 15 15>, angularLimitsMin=<-15 -15 0>, angularTarget=<1 0 0>}
      localJointOffset=<4 0 0>
     }]
   }
   n30 = Anim Dynamics - Bone: dyn_coat_shoulder_01_r(ComponentPose=@n29, Alpha=1, ExternalForce="0,0,0") {
    angularDampingOverride=1
    angularSpringConstant=20000
    bAngularSpring=true
    bOverrideAngularDamping=true
    bUseGravityOverride=true
    gravityScale=0.35
    simulationSpace=RootRelative
    boundBone {boneName=dyn_coat_shoulder_01_r}
    componentLinearAccScale=<1 1 1>
    componentLinearVelScale=<1 1 1>
    gravityOverride {z=-980}
    physicsBodyDefinitions [{
      boundBone {boneName=dyn_coat_shoulder_01_r}
      boxExtents=<20 30 30>
      constraintSetup {angularLimitsMax=<15 15 15>, angularLimitsMin=<-15 -15 -15>, angularTarget=<1 0 0>}
      localJointOffset=<-4 0 0>
     }]
   }
   n31 = "Transform (Modify) Bone - Bone: dyn_coat_shoulder_01_r"!(ComponentPose=@n30, bAlphaBoolEnabled:bool=True, Translation="0,0,0", Rotation="0,-20,0", Scale="1,1,1") -> (Pose) {
    alphaInputType=Bool
    rotationMode=BMM_Additive
    rotationSpace=BCS_BoneSpace
    alphaBoolBlend {blendInTime=0.2, blendOutTime=0.2}
    boneToModify {boneName=dyn_coat_shoulder_01_r}
    rotation {yaw=-20}
   }
   Save cached pose 'Post Main Dynamics'(Pose=Component To Local(ComponentPose=@n31)) {cacheName=Post Main Dynamics}
  ]}
 EventGraph {nodes [
   n1 = Event BlueprintUpdateAnimation()
   > Set DeltaTimeLocal(DeltaTimeLocal:double=@n1.DeltaTimeX) -> (Output_Get:double)
   > n2 = Sequence!(execute) -> (then_0, then_1, then_2:Exec)
   > n3 = Branch(Condition=Get bIsInFirstPerson!(self:Fort Player Anim Instance Object Reference=GetOwnerFortPlayerAnimInstance!(self:Custom Character Part Anim Instance Object Reference) -> (ReturnValue:Fort Player Anim Instance Object Reference)) -> (bIsInFirstPerson:bool))
   > n4 = Branch(Condition=Get First Person Mode Enabled() -> (First Person Mode Enabled:bool))
   n5 = GetPartSkeletalMeshComponent(self=Self-Reference())
   Set ClothBlendWeight!(execute=@n4.else, ClothBlendWeight:float=0, self:Skeletal Mesh Component Object Reference=@n5) -> (then, Output_Get:float)
   > Set bDisableClothSimulation!(execute, bDisableClothSimulation:bool=true, self:Skeletal Mesh Component Object Reference=@n5) -> (then, Output_Get:bool)
   > Set First Person Mode Enabled(First Person Mode Enabled:bool=true) -> (Output_Get:bool)
   Branch(execute=@n3.else, Condition=Get First Person Mode Enabled())
   > Set ClothBlendWeight(ClothBlendWeight=1, self=@n5)
   > Set bDisableClothSimulation(bDisableClothSimulation=false, self=@n5)
   > Set First Person Mode Enabled(First Person Mode Enabled=false)
   Branch(execute=@n2.then_1, Condition=Get bClothEnabled())
   > n6 = "Cloth Controller Function (Parameter Asset) (Nativized)"!(
    execute
    CharacterPartAnimInstance:Custom Character Part Anim Instance Object Reference=Self-Reference()
    ClothingAssetName:Name=M_MED_NoonDoctor_Clothing
    ClothCoreSettings:"Cloth Core Settings Structure (by ref)"=Get Cloth Core Settings_Runtime() -> (Cloth Core Settings_Runtime:Cloth Core Settings Structure)
    WindGustLocalSettings:"Wind Gust Settings Structure (by ref)"=Get Wind Gust Local Settings Out() -> (Wind Gust Local Settings Out:Wind Gust Settings Structure)
    ClothUpAxis:Vector="1,0,0"
    LinVelScaleFwdBwdClamp:float=0
    LinVelScaleUpDwnClamp:float=0.7
    DeltaTime:float=Get DeltaTimeLocal() -> (DeltaTimeLocal:double)
    GameTimeInSeconds:float=GetGameTimeInSeconds()
    StateLogicLayerName:Name=none
    Parameters:Fort Character Dynamics Parameters Object Reference=&/M_MED_NoonDoctor_Physics_parameters
   ) -> (
    then
    OutCurrentState:Name=None
    OutClothParameters:Cloth Parameters Structure
    OutClothCoreSettings:Cloth Core Settings Structure
    OutWindGustSettings:Wind Gust Parameters Structure
    OutWindGustLocalSettings:Wind Gust Settings Structure
    OutWindVelocity:Vector
    OutLinearVelVector:Vector
   )
   > Set Wind Gust Local Settings Out(Wind Gust Local Settings Out:Wind Gust Settings Structure=@n6.OutWindGustLocalSettings) -> (Output_Get:Wind Gust Settings Structure)
   Branch(execute=@n2.then_2, Condition=Get bRigidBodyEnabled())
   > n7 = "Gravity Override State Function (Parameter Asset) (Nativized)"(
    CharacterPartAnimInstance=Self-Reference()
    GameTimeInSeconds=GetGameTimeInSeconds()
    Parameters=&M_MED_NoonDoctor_Physics_parameters
   )
   > Set GravityOverride(GravityOverride=@n7)
   > Set RBAN Settings Out(RBAN Settings Out=@n7.OutRigidBodyAnimNodeParameters)
   n8 = Event OnCharacterCustomizationLoaded()
   > Set Custom Base Mesh(Custom Base Mesh:Skeletal Mesh Component Object Reference=@n8.CustomBaseMesh) -> (Output_Get:Skeletal Mesh Component Object Reference)
   > n9 = Branch(Condition=CheckForBackbling().bIsBackblingACape)
   n10 = GetPartSkeletalMeshComponent(self=Self-Reference())
   > Set ClothBlendWeight(ClothBlendWeight=0, self=@n10)
   > Set bDisableClothSimulation(bDisableClothSimulation=true, self=@n10)
   Set ClothBlendWeight(execute=@n9.else, ClothBlendWeight=1, self=@n10)
   > Set bDisableClothSimulation(bDisableClothSimulation=false, self=@n10)
   Event OnEndTargeting() {is=title K2Node_Event}
   > Set Shoulder Flap Dynamics Weight(Shoulder Flap Dynamics Weight:double=1) -> (Output_Get:double)
   Event OnBeginTargeting() {is=title K2Node_Event}
   > Set Shoulder Flap Dynamics Weight(Shoulder Flap Dynamics Weight=0.5)
   Get Cloth Core Settings_Default() -> (Cloth Core Settings_Default:Cloth Core Settings Structure)
  ]}
]
variables [
 Cloth Core Settings Runtime {
  type=ClothCoreSettings
  default {
   collisionThickness=0
   dampingCoefficient=0.05
   frictionCoefficient=0.3
   localDampingCoefficient=0
   numberIterations=2
   numberSubsteps=1
   selfCollisionThickness=0
   animDriveDamping=<0 1>
   animDriveStiffness=<0 1>
   areaStiffness=<1 1>
   bendingStiffness=<1 1>
   edgeStiffness=<1 1>
   tetherStiffness=<1 1>
  }
 }
 Cloth Core Settings_Default {
  type=ClothCoreSettings
  default {
   collisionThickness=0
   dampingCoefficient=0.05
   frictionCoefficient=0.3
   localDampingCoefficient=0
   numberIterations=2
   numberSubsteps=1
   selfCollisionThickness=0
   animDriveDamping=<0 1>
   animDriveStiffness=<0 1>
   areaStiffness=<1 1>
   bendingStiffness=<1 1>
   edgeStiffness=<1 1>
   tetherStiffness=<1 1>
  }
 }
 Cloth Core Settings_Runtime {
  type=ClothCoreSettings
  default {
   collisionThickness=0.5
   dampingCoefficient=0.1
   frictionCoefficient=0.3
   localDampingCoefficient=0
   numberIterations=2
   numberSubsteps=1
   selfCollisionThickness=0
   animDriveDamping=<0 1>
   animDriveStiffness=<0.15 1>
   areaStiffness=<1 1>
   bendingStiffness=<1 1>
   edgeStiffness=<1 1>
   tetherStiffness=<1 1>
  }
 }
 Custom Base Mesh {type=SkeletalMeshComponent}
 DeltaTimeLocal {type=number}
 First Person Mode Enabled {type=boolean}
 GravityOverride {type=Vector}
 RBAN Settings Out {type=FortRigidBodyAnimNodeParameters}
 Shoulder Flap Dynamics Weight {type=number}
 Wind Gust Local Settings Native {type=WindGustSettings, default {elapsedTime=0, finalGustStrengthLocal=0, gustDuration=0, gustStrength=0, timeBetweenGusts=0}}
 Wind Gust Local Settings Out {type=WindGustSettings, default {elapsedTime=0, finalGustStrengthLocal=0, gustDuration=0, gustStrength=0, timeBetweenGusts=0}}
]
