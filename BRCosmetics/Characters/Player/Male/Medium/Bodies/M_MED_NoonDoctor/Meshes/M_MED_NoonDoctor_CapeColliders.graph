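# Physics asset for the male medium NoonDoctor character body, supplying collision shapes that, by its name, serve as colliders for the cape.
# Capsules cover the pelvis, spine, clavicles and neck, with a sphere on the head; arm, leg and foot bodies have no shapes.
# Three rear coat bone chains (segments 02–05) carry large capsules offset sideways from their bones.
# It defines no constraints, so it holds collision volumes rather than a jointed ragdoll.
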PhysicsAsset /BRCosmetics/Characters/Player/Male/Medium/Bodies/M_MED_NoonDoctor/Meshes/M_MED_NoonDoctor_CapeColliders
bodies {
 calf_l []
 calf_r []
 clavicle_l [{shapeType=Capsule, length=9.0192, radius=9.8871, center=<9.4266 -1.1806 -2.4977>, rotation=<rot 69.3297 1.4476 176.9462>}]
 clavicle_r [{shapeType=Capsule, length=9.0192, radius=9.8871, center=<-9.4266 1.1806 2.4977>, rotation=<rot 69.3297 1.4476 -3.0538>}]
 dyn_coat_rear_02 [{shapeType=Capsule, length=7, radius=16.5312, shapeName=dyn_coat_rear_02_capsule, center=<3.0476 -13.4541 0>}]
 dyn_coat_rear_02_r [{shapeType=Capsule, length=7, radius=16.5312, shapeName=dyn_coat_rear_02_r_capsule, center=<-3.0476 13.4782 0>}]
 dyn_coat_rear_03 [{shapeType=Capsule, length=7.4688, radius=16.5312, shapeName=dyn_coat_rear_03_capsule, center=<3.0476 -13.4028 0>}]
 dyn_coat_rear_03_r [{shapeType=Capsule, length=7.4688, radius=16.5312, shapeName=dyn_coat_rear_03_r_capsule, center=<-3.0476 13.4178 0>}]
 dyn_coat_rear_04 [{shapeType=Capsule, length=7, radius=16.5312, shapeName=dyn_coat_rear_04_capsule, center=<3.0476 -13.4541 0>}]
 dyn_coat_rear_04_r [{shapeType=Capsule, length=7, radius=16.5312, shapeName=dyn_coat_rear_04_r_capsule, center=<-3.0476 13.4782 0>}]
 dyn_coat_rear_05 [{shapeType=Capsule, length=7.4688, radius=16.5312, shapeName=dyn_coat_rear_05_capsule, center=<3.0476 -13.4028 0>}]
 dyn_coat_rear_05_r [{shapeType=Capsule, length=7.4688, radius=16.5312, shapeName=dyn_coat_rear_05_r_capsule, center=<-3.0476 13.4178 0>}]
 dyn_coat_rear_l_02 [{shapeType=Capsule, length=7, radius=16.5312, shapeName=dyn_coat_rear_l_02_capsule, center=<3.0476 -13.4541 0>}]
 dyn_coat_rear_l_03 [{shapeType=Capsule, length=7.4688, radius=16.5312, shapeName=dyn_coat_rear_l_03_capsule, center=<3.0476 -13.4028 0>}]
 dyn_coat_rear_l_04 [{shapeType=Capsule, length=7, radius=16.5312, shapeName=dyn_coat_rear_l_04_capsule, center=<3.0476 -13.4541 0>}]
 dyn_coat_rear_l_05 [{shapeType=Capsule, length=7.4688, radius=16.5312, shapeName=dyn_coat_rear_l_05_capsule, center=<3.0476 -13.4028 0>}]
 foot_l []
 foot_r []
 head [{shapeType=Sphere, radius=17.3125, center=<11.028 -0.174 0>}]
 lowerarm_l []
 lowerarm_r []
 neck_02 [{shapeType=Capsule, length=13.4375, radius=7.3175, center=<-0.2642 -0.3408 0>, rotation=<rot 90 -171.1193 -175.3978>}]
 pelvis [{shapeType=Capsule, length=8.6456, radius=14.5388, center=<-2.4887 0.1881 0>}]
 spine_01 [{shapeType=Capsule, length=7.9438, radius=13.5586, center=<-1.7965 0.1884 0>}]
 spine_02 [
  {shapeType=Capsule, length=7.3264, radius=13.6322, center=<-1.711 -0.1851 0.0001>}
  {shapeType=Capsule, length=29.3956, radius=11.5807, center=<2.5024 0.4992 3.9825>, rotation=<rot 90 0.0049 0.0049>}
  {shapeType=Capsule, length=29.875, radius=11.1013, center=<2.4704 0.4991 -3.9825>, rotation=<rot 90 -0.0396 -0.0396>}
 ]
 spine_03 [{shapeType=Capsule, length=12.3092, radius=18.3074, center=<-0.0135 -2.5525 0.0001>}]
 spine_04 [{shapeType=Capsule, length=8.4589, radius=15.114, center=<4.0137 -1.0609 0.0002>, rotation=<rot 0 14.2188 0>}]
 spine_05 [
  {shapeType=Capsule, length=8.0226, radius=15.695, center=<-2.2795 1.8109 0>}
  {shapeType=Capsule, length=13.9518, radius=13.9425, center=<-0.2916 0.4452 0>}
 ]
 thigh_l []
 thigh_r []
 upperarm_l []
 upperarm_r []
}
constraints []
