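# Physics asset providing cloth collision volumes for the NoonDoctor medium male player cosmetic body, as its name indicates.
# Single capsules approximate the pelvis, two spine bones, upper arms, thighs, calves and feet, with an extra capsule on a dynamic coat tassel bone.
# Leg capsules mirror left to right, while the two upper arm capsules differ in size and placement.
# It defines no constraints between bodies.
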PhysicsAsset /BRCosmetics/Characters/Player/Male/Medium/Bodies/M_MED_NoonDoctor/Meshes/M_MED_NoonDoctor_ClothColliders
bodies {
 calf_l [{shapeType=Capsule, length=32.0675, radius=10.2478, center=<-17.7339 -2.204 2.395>, rotation=<rot 0 -92.6095 -88.1249>}]
 calf_r [{shapeType=Capsule, length=32.0675, radius=10.2478, center=<17.7339 2.204 -2.395>, rotation=<rot 0 -92.6095 91.8751>}]
 dyn_coat_tassle_02 [{
   shapeType=Capsule
   length=4.0508
   radius=3.864
   shapeName=dyn_coat_tassle_02_capsule
   center=<1.943 1.2283 -0.0568>
   rotation=<rot 0.6859 -121.6614 -86.3595>
  }]
 foot_l [{shapeType=Capsule, length=10.9635, radius=11.6802, center=<-2.8687 6.6574 -0.2921>, rotation=<rot 0 11.9606 -89.8608>}]
 foot_r [{shapeType=Capsule, length=10.9635, radius=11.6802, center=<2.8687 -6.6574 0.2921>, rotation=<rot 0 11.9606 90.1392>}]
 pelvis [{shapeType=Capsule, length=13.9163, radius=12.7914, center=<-4.8419 -2.0272 0>, rotation=<rot 0 90 0>}]
 spine_02 [{shapeType=Capsule, length=8.8262, radius=11.8218, shapeName=spine_02_capsule, center=<13.641 0 0>, rotation=<rot 0 90 90>}]
 spine_04 [{shapeType=Capsule, length=9.3499, radius=10.5291, shapeName=spine_04_capsule, center=<-19.411 -0.7853 0.0226>, rotation=<rot 90 3.5493 0>}]
 thigh_l [{
   shapeType=Capsule
   length=26.868
   radius=11.9112
   shapeName=thigh_l_capsule
   center=<-18.5574 -1.8565 -0.7536>
   rotation=<rot 0.7821 -90.0018 -82.2254>
  }]
 thigh_r [{
   shapeType=Capsule
   length=26.868
   radius=11.9112
   shapeName=thigh_r_capsule
   center=<18.5574 1.8565 0.7536>
   rotation=<rot 0.7821 -90.0018 97.7746>
  }]
 upperarm_l [{shapeType=Capsule, length=14.858, radius=5.9515, shapeName=upperarm_l_capsule, center=<12.7829 0.5201 -0.7288>, rotation=<rot 90 4.4495 0>}]
 upperarm_r [{
   shapeType=Capsule
   length=13.858
   radius=6.9515
   shapeName=upperarm_r_capsule
   center=<-9.3842 1.3626 1.5801>
   rotation=<rot 87.1875 -175.5505 0>
  }]
}
constraints []
