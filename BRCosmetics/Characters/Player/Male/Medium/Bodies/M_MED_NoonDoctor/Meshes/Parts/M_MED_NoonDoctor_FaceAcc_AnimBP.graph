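# Animation Blueprint for the NoonDoctor character's face-accessory mesh, adding secondary motion to its hair.
# It copies the attached parent's pose, then runs a serial chain of single-bone Anim Dynamics nodes with angular springs, limits and overridden gravity.
# Bun, flyaways and left, right and middle strands have separate tuning; most strands blend partially, and the right and middle group bones have zero alpha.
# The event graph has no logic.
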AnimBlueprint /BRCosmetics/Characters/Player/Male/Medium/Bodies/M_MED_NoonDoctor/Meshes/Parts/M_MED_NoonDoctor_FaceAcc_AnimBP
target_skeleton=&/M_MED_NoonDoctor_FaceAcc_Skeleton
graphs [
 AnimGraph {nodes [
   n1 = Anim Dynamics - Bone: dyn_hair_bun_01(ComponentPose=Local To Component(LocalPose=Copy Pose From Mesh() {bUseAttachedParent=true}), Alpha=1, ExternalForce="0,0,0") {
    angularDampingOverride=0.94
    angularSpringConstant=200
    bAngularSpring=true
    bOverrideAngularDamping=true
    bUseGravityOverride=true
    gravityScale=0.35
    simulationSpace=RootRelative
    boundBone {boneName=dyn_hair_bun_01}
    componentAppliedLinearAccClamp=<0 0 0>
    gravityOverride {z=-1400}
    physicsBodyDefinitions [{
      boundBone {boneName=dyn_hair_bun_01}
      boxExtents=<15 10 15>
      constraintSetup {angularLimitsMax=<5 5 5>, angularLimitsMin=<-5 -5 -5>, angularTarget=<1 0 0>}
      localJointOffset=<4 0 0>
     }]
    relativeSpaceBone {boneName=head}
    simSpaceSettings {maxAngularAcceleration=140, maxAngularVelocity=8, simSpaceAngularAlpha=0.4}
   }
   n2 = Anim Dynamics - Bone: dyn_hair_bun_02(ComponentPose=@n1, Alpha=1, ExternalForce="0,0,0") {
    angularDampingOverride=0.94
    angularSpringConstant=200
    bAngularSpring=true
    bOverrideAngularDamping=true
    bUseGravityOverride=true
    gravityScale=0.35
    simulationSpace=RootRelative
    boundBone {boneName=dyn_hair_bun_02}
    componentAppliedLinearAccClamp=<0 0 0>
    gravityOverride {z=-1400}
    physicsBodyDefinitions [{
      boundBone {boneName=dyn_hair_bun_02}
      boxExtents=<20 20 20>
      constraintSetup {angularLimitsMax=<5 5 5>, angularLimitsMin=<-5 -5 -5>, angularTarget=<1 0 0>}
      localJointOffset=<3 0 0>
     }]
    relativeSpaceBone {boneName=head}
    simSpaceSettings {maxAngularAcceleration=140, maxAngularVelocity=8, simSpaceAngularAlpha=0.5}
   }
   n3 = Anim Dynamics - Bone: dyn_hair_bun_flyaway_01(ComponentPose=@n2, Alpha=1, ExternalForce="0,0,0") {
    angularDampingOverride=0.99
    angularSpringConstant=20000
    bAngularSpring=true
    bOverrideAngularDamping=true
    bUseGravityOverride=true
    gravityScale=0.35
    simulationSpace=RootRelative
    boundBone {boneName=dyn_hair_bun_flyaway_01}
    componentAppliedLinearAccClamp=<0 0 0>
    gravityOverride {z=-980}
    physicsBodyDefinitions [{
      boundBone {boneName=dyn_hair_bun_flyaway_01}
      boxExtents=<20 30 30>
      constraintSetup {angularLimitsMax=<15 15 15>, angularLimitsMin=<-15 -15 -15>, angularTarget=<1 0 0>}
      localJointOffset=<3 0 0>
     }]
    relativeSpaceBone {boneName=spine_04}
    simSpaceSettings {maxAngularAcceleration=300, maxAngularVelocity=12, simSpaceAngularAlpha=1}
   }
   n4 = Anim Dynamics - Bone: dyn_hair_bun_flyaway_02(ComponentPose=@n3, Alpha=1, ExternalForce="0,0,0") {
    angularDampingOverride=0.99
    angularSpringConstant=20000
    bAngularSpring=true
    bOverrideAngularDamping=true
    bUseGravityOverride=true
    gravityScale=0.35
    simulationSpace=RootRelative
    boundBone {boneName=dyn_hair_bun_flyaway_02}
    componentAppliedLinearAccClamp=<0 0 0>
    gravityOverride {z=-980}
    physicsBodyDefinitions [{
      boundBone {boneName=dyn_hair_bun_flyaway_02}
      boxExtents=<20 30 30>
      constraintSetup {angularLimitsMax=<15 15 15>, angularLimitsMin=<-15 -15 -15>, angularTarget=<1 0 0>}
      localJointOffset=<3 0 0>
     }]
    relativeSpaceBone {boneName=spine_04}
    simSpaceSettings {maxAngularAcceleration=300, maxAngularVelocity=12, simSpaceAngularAlpha=1}
   }
   n5 = Anim Dynamics - Bone: dyn_hair_bun_flyaway_03(ComponentPose=@n4, Alpha=1, ExternalForce="0,0,0") {
    angularDampingOverride=0.99
    angularSpringConstant=20000
    bAngularSpring=true
    bOverrideAngularDamping=true
    bUseGravityOverride=true
    gravityScale=0.35
    simulationSpace=RootRelative
    boundBone {boneName=dyn_hair_bun_flyaway_03}
    componentAppliedLinearAccClamp=<0 0 0>
    gravityOverride {z=-980}
    physicsBodyDefinitions [{
      boundBone {boneName=dyn_hair_bun_flyaway_03}
      boxExtents=<20 30 30>
      constraintSetup {angularLimitsMax=<15 15 15>, angularLimitsMin=<-15 -15 -15>, angularTarget=<1 0 0>}
      localJointOffset=<3 0 0>
     }]
    relativeSpaceBone {boneName=spine_04}
    simSpaceSettings {maxAngularAcceleration=300, maxAngularVelocity=12, simSpaceAngularAlpha=1}
   }
   n6 = Anim Dynamics - Bone: dyn_hair_bun_flyaway_04(ComponentPose=@n5, Alpha=1, ExternalForce="0,0,0") {
    angularDampingOverride=0.99
    angularSpringConstant=20000
    bAngularSpring=true
    bOverrideAngularDamping=true
    bUseGravityOverride=true
    gravityScale=0.35
    simulationSpace=RootRelative
    boundBone {boneName=dyn_hair_bun_flyaway_04}
    componentAppliedLinearAccClamp=<0 0 0>
    gravityOverride {z=-980}
    physicsBodyDefinitions [{
      boundBone {boneName=dyn_hair_bun_flyaway_04}
      boxExtents=<20 30 30>
      constraintSetup {angularLimitsMax=<15 15 15>, angularLimitsMin=<-15 -15 -15>, angularTarget=<1 0 0>}
      localJointOffset=<3 0 0>
     }]
    relativeSpaceBone {boneName=spine_04}
    simSpaceSettings {maxAngularAcceleration=300, maxAngularVelocity=12, simSpaceAngularAlpha=1}
   }
   n7 = Anim Dynamics - Bone: dyn_hair_bun_flyaway_05(ComponentPose=@n6, Alpha=1, ExternalForce="0,0,0") {
    angularDampingOverride=0.99
    angularSpringConstant=20000
    bAngularSpring=true
    bOverrideAngularDamping=true
    bUseGravityOverride=true
    gravityScale=0.35
    simulationSpace=RootRelative
    boundBone {boneName=dyn_hair_bun_flyaway_05}
    componentAppliedLinearAccClamp=<0 0 0>
    gravityOverride {z=-980}
    physicsBodyDefinitions [{
      boundBone {boneName=dyn_hair_bun_flyaway_05}
      boxExtents=<20 30 30>
      constraintSetup {angularLimitsMax=<15 15 15>, angularLimitsMin=<-15 -15 -15>, angularTarget=<1 0 0>}
      localJointOffset=<3 0 0>
     }]
    relativeSpaceBone {boneName=spine_04}
    simSpaceSettings {maxAngularAcceleration=300, maxAngularVelocity=12, simSpaceAngularAlpha=1}
   }
   n8 = Anim Dynamics - Bone: dyn_hair_bun_flyaway_06(ComponentPose=@n7, Alpha=1, ExternalForce="0,0,0") {
    angularDampingOverride=0.99
    angularSpringConstant=20000
    bAngularSpring=true
    bOverrideAngularDamping=true
    bUseGravityOverride=true
    gravityScale=0.35
    simulationSpace=RootRelative
    boundBone {boneName=dyn_hair_bun_flyaway_06}
    componentAppliedLinearAccClamp=<0 0 0>
    gravityOverride {z=-980}
    physicsBodyDefinitions [{
      boundBone {boneName=dyn_hair_bun_flyaway_06}
      boxExtents=<20 30 30>
      constraintSetup {angularLimitsMax=<15 15 15>, angularLimitsMin=<-15 -15 -15>, angularTarget=<1 0 0>}
      localJointOffset=<3 0 0>
     }]
    relativeSpaceBone {boneName=spine_04}
    simSpaceSettings {maxAngularAcceleration=300, maxAngularVelocity=12, simSpaceAngularAlpha=1}
   }
   n9 = Anim Dynamics - Bone: dyn_hair_bun_flyaway_07(ComponentPose=@n8, Alpha=1, ExternalForce="0,0,0") {
    angularDampingOverride=0.99
    angularSpringConstant=20000
    bAngularSpring=true
    bOverrideAngularDamping=true
    bUseGravityOverride=true
    gravityScale=0.35
    simulationSpace=RootRelative
    boundBone {boneName=dyn_hair_bun_flyaway_07}
    componentAppliedLinearAccClamp=<0 0 0>
    gravityOverride {z=-980}
    physicsBodyDefinitions [{
      boundBone {boneName=dyn_hair_bun_flyaway_07}
      boxExtents=<20 30 30>
      constraintSetup {angularLimitsMax=<15 15 15>, angularLimitsMin=<-15 -15 -15>, angularTarget=<1 0 0>}
      localJointOffset=<3 0 0>
     }]
    relativeSpaceBone {boneName=spine_04}
    simSpaceSettings {maxAngularAcceleration=300, maxAngularVelocity=12, simSpaceAngularAlpha=1}
   }
   n10 = Anim Dynamics - Bone: dyn_hair_bun_flyaway_08(ComponentPose=@n9, Alpha=1, ExternalForce="0,0,0") {
    angularDampingOverride=0.99
    angularSpringConstant=20000
    bAngularSpring=true
    bOverrideAngularDamping=true
    bUseGravityOverride=true
    gravityScale=0.35
    simulationSpace=RootRelative
    boundBone {boneName=dyn_hair_bun_flyaway_08}
    componentAppliedLinearAccClamp=<0 0 0>
    gravityOverride {z=-980}
    physicsBodyDefinitions [{
      boundBone {boneName=dyn_hair_bun_flyaway_08}
      boxExtents=<20 30 30>
      constraintSetup {angularLimitsMax=<15 15 15>, angularLimitsMin=<-15 -15 -15>, angularTarget=<1 0 0>}
      localJointOffset=<3 0 0>
     }]
    relativeSpaceBone {boneName=spine_04}
    simSpaceSettings {maxAngularAcceleration=300, maxAngularVelocity=12, simSpaceAngularAlpha=1}
   }
   n11 = Anim Dynamics - Bone: dyn_hair_strands_l!(ComponentPose=@n10, ExternalForce="0,0,0") -> (Pose) {
    alpha=0.7
    angularDampingOverride=0.99
    angularSpringConstant=5000
    bAngularSpring=true
    bOverrideAngularDamping=true
    bUseGravityOverride=true
    gravityScale=0.35
    simulationSpace=RootRelative
    boundBone {boneName=dyn_hair_strands_l}
    componentAppliedLinearAccClamp=<0 0 0>
    gravityOverride {z=-980}
    physicsBodyDefinitions [{
      boundBone {boneName=dyn_hair_strands_l}
      boxExtents=<15 20 20>
      constraintSetup {angularLimitsMax=<5 5 5>, angularLimitsMin=<-5 0 -5>, angularTarget=<1 0 0>}
      localJointOffset=<2 0 0>
     }]
    relativeSpaceBone {boneName=root}
    simSpaceSettings {maxAngularAcceleration=300, maxAngularVelocity=12, simSpaceAngularAlpha=0.4}
   }
   n12 = Anim Dynamics - Bone: dyn_hair_strand_a_l_01!(ComponentPose=@n11, ExternalForce="0,0,0") -> (Pose) {
    alpha=0.7
    angularDampingOverride=0.99
    angularSpringConstant=5000
    bAngularSpring=true
    bOverrideAngularDamping=true
    bUseGravityOverride=true
    gravityScale=0.35
    simulationSpace=RootRelative
    boundBone {boneName=dyn_hair_strand_a_l_01}
    componentAppliedLinearAccClamp=<0 0 0>
    gravityOverride {z=-980}
    physicsBodyDefinitions [{
      boundBone {boneName=dyn_hair_strand_a_l_01}
      boxExtents=<15 20 20>
      constraintSetup {angularLimitsMax=<5 5 5>, angularLimitsMin=<-5 0 -5>, angularTarget=<1 0 0>}
      localJointOffset=<2 0 0>
     }]
    relativeSpaceBone {boneName=root}
    simSpaceSettings {maxAngularAcceleration=300, maxAngularVelocity=12, simSpaceAngularAlpha=0.4}
   }
   n13 = Anim Dynamics - Bone: dyn_hair_strand_a_l_02!(ComponentPose=@n12, ExternalForce="0,0,0") -> (Pose) {
    alpha=0.7
    angularDampingOverride=0.99
    angularSpringConstant=5000
    bAngularSpring=true
    bOverrideAngularDamping=true
    bUseGravityOverride=true
    gravityScale=0.35
    simulationSpace=RootRelative
    boundBone {boneName=dyn_hair_strand_a_l_02}
    componentAppliedLinearAccClamp=<0 0 0>
    gravityOverride {z=-980}
    physicsBodyDefinitions [{
      boundBone {boneName=dyn_hair_strand_a_l_02}
      boxExtents=<15 20 20>
      constraintSetup {angularLimitsMax=<10 10 10>, angularLimitsMin=<-10 0 -10>, angularTarget=<1 0 0>}
      localJointOffset=<2 0 0>
     }]
    relativeSpaceBone {boneName=root}
    simSpaceSettings {maxAngularAcceleration=300, maxAngularVelocity=12, simSpaceAngularAlpha=0.4}
   }
   n14 = Anim Dynamics - Bone: dyn_hair_strand_b_l_01!(ComponentPose=@n13, ExternalForce="0,0,0") -> (Pose) {
    alpha=0.7
    angularDampingOverride=0.99
    angularSpringConstant=5000
    bAngularSpring=true
    bOverrideAngularDamping=true
    bUseGravityOverride=true
    gravityScale=0.35
    simulationSpace=RootRelative
    boundBone {boneName=dyn_hair_strand_b_l_01}
    componentAppliedLinearAccClamp=<0 0 0>
    gravityOverride {z=-980}
    physicsBodyDefinitions [{
      boundBone {boneName=dyn_hair_strand_b_l_01}
      boxExtents=<15 20 20>
      constraintSetup {angularLimitsMax=<5 5 5>, angularLimitsMin=<-5 0 -5>, angularTarget=<1 0 0>}
      localJointOffset=<2 0 0>
     }]
    relativeSpaceBone {boneName=root}
    simSpaceSettings {maxAngularAcceleration=300, maxAngularVelocity=12, simSpaceAngularAlpha=0.4}
   }
   n15 = Anim Dynamics - Bone: dyn_hair_strand_b_l_02!(ComponentPose=@n14, ExternalForce="0,0,0") -> (Pose) {
    alpha=0.7
    angularDampingOverride=0.99
    angularSpringConstant=5000
    bAngularSpring=true
    bOverrideAngularDamping=true
    bUseGravityOverride=true
    gravityScale=0.35
    simulationSpace=RootRelative
    boundBone {boneName=dyn_hair_strand_b_l_02}
    componentAppliedLinearAccClamp=<0 0 0>
    gravityOverride {z=-980}
    physicsBodyDefinitions [{
      boundBone {boneName=dyn_hair_strand_b_l_02}
      boxExtents=<15 20 20>
      constraintSetup {angularLimitsMax=<10 10 10>, angularLimitsMin=<-10 0 -10>, angularTarget=<1 0 0>}
      localJointOffset=<2 0 0>
     }]
    relativeSpaceBone {boneName=root}
    simSpaceSettings {maxAngularAcceleration=300, maxAngularVelocity=12, simSpaceAngularAlpha=0.4}
   }
   n16 = Anim Dynamics - Bone: dyn_hair_strand_b_l_03!(ComponentPose=@n15, ExternalForce="0,0,0") -> (Pose) {
    alpha=0.9
    angularDampingOverride=0.99
    angularSpringConstant=5000
    bAngularSpring=true
    bOverrideAngularDamping=true
    bUseGravityOverride=true
    gravityScale=0.35
    simulationSpace=RootRelative
    boundBone {boneName=dyn_hair_strand_b_l_03}
    componentAppliedLinearAccClamp=<0 0 0>
    gravityOverride {z=-980}
    physicsBodyDefinitions [{
      boundBone {boneName=dyn_hair_strand_b_l_03}
      boxExtents=<15 20 20>
      constraintSetup {angularLimitsMax=<15 15 15>, angularLimitsMin=<-15 0 -15>, angularTarget=<1 0 0>}
      localJointOffset=<2 0 0>
     }]
    relativeSpaceBone {boneName=root}
    simSpaceSettings {maxAngularAcceleration=300, maxAngularVelocity=12, simSpaceAngularAlpha=0.4}
   }
   n17 = Anim Dynamics - Bone: dyn_hair_strand_c_l_01!(ComponentPose=@n16, ExternalForce="0,0,0") -> (Pose) {
    alpha=0.7
    angularDampingOverride=0.99
    angularSpringConstant=5000
    bAngularSpring=true
    bOverrideAngularDamping=true
    bUseGravityOverride=true
    gravityScale=0.35
    simulationSpace=RootRelative
    boundBone {boneName=dyn_hair_strand_c_l_01}
    componentAppliedLinearAccClamp=<0 0 0>
    gravityOverride {z=-980}
    physicsBodyDefinitions [{
      boundBone {boneName=dyn_hair_strand_c_l_01}
      boxExtents=<15 20 20>
      constraintSetup {angularLimitsMax=<5 5 5>, angularLimitsMin=<-5 0 -5>, angularTarget=<1 0 0>}
      localJointOffset=<2 0 0>
     }]
    relativeSpaceBone {boneName=root}
    simSpaceSettings {maxAngularAcceleration=300, maxAngularVelocity=12, simSpaceAngularAlpha=0.4}
   }
   n18 = Anim Dynamics - Bone: dyn_hair_strand_c_l_02!(ComponentPose=@n17, ExternalForce="0,0,0") -> (Pose) {
    alpha=0.7
    angularDampingOverride=0.99
    angularSpringConstant=5000
    bAngularSpring=true
    bOverrideAngularDamping=true
    bUseGravityOverride=true
    gravityScale=0.35
    simulationSpace=RootRelative
    boundBone {boneName=dyn_hair_strand_c_l_02}
    componentAppliedLinearAccClamp=<0 0 0>
    gravityOverride {z=-980}
    physicsBodyDefinitions [{
      boundBone {boneName=dyn_hair_strand_c_l_02}
      boxExtents=<15 20 20>
      constraintSetup {angularLimitsMax=<10 10 10>, angularLimitsMin=<-10 0 -10>, angularTarget=<1 0 0>}
      localJointOffset=<2 0 0>
     }]
    relativeSpaceBone {boneName=root}
    simSpaceSettings {maxAngularAcceleration=300, maxAngularVelocity=12, simSpaceAngularAlpha=0.4}
   }
   n19 = Anim Dynamics - Bone: dyn_hair_strand_c_l_03!(ComponentPose=@n18, ExternalForce="0,0,0") -> (Pose) {
    alpha=0.7
    angularDampingOverride=0.99
    angularSpringConstant=5000
    bAngularSpring=true
    bOverrideAngularDamping=true
    bUseGravityOverride=true
    gravityScale=0.35
    simulationSpace=RootRelative
    boundBone {boneName=dyn_hair_strand_c_l_03}
    componentAppliedLinearAccClamp=<0 0 0>
    gravityOverride {z=-980}
    physicsBodyDefinitions [{
      boundBone {boneName=dyn_hair_strand_c_l_03}
      boxExtents=<15 20 20>
      constraintSetup {angularLimitsMax=<15 15 15>, angularLimitsMin=<-15 0 -15>, angularTarget=<1 0 0>}
      localJointOffset=<2 0 0>
     }]
    relativeSpaceBone {boneName=root}
    simSpaceSettings {maxAngularAcceleration=300, maxAngularVelocity=12, simSpaceAngularAlpha=0.4}
   }
   n20 = Anim Dynamics - Bone: dyn_hair_strands_r!(ComponentPose=@n19, ExternalForce="0,0,0") -> (Pose) {
    alpha=0
    angularDampingOverride=0.999
    angularSpringConstant=7000
    bAngularSpring=true
    bOverrideAngularDamping=true
    bUseGravityOverride=true
    gravityScale=0.35
    simulationSpace=RootRelative
    boundBone {boneName=dyn_hair_strands_r}
    componentAppliedLinearAccClamp=<0 0 0>
    gravityOverride {z=-980}
    physicsBodyDefinitions [{
      boundBone {boneName=dyn_hair_strands_r}
      boxExtents=<15 20 20>
      constraintSetup {angularLimitsMax=<5 5 5>, angularLimitsMin=<-5 0 -5>, angularTarget=<1 0 0>}
      localJointOffset=<2 0 0>
     }]
    relativeSpaceBone {boneName=root}
    simSpaceSettings {maxAngularAcceleration=300, maxAngularVelocity=12, simSpaceAngularAlpha=0.4}
   }
   n21 = Anim Dynamics - Bone: dyn_hair_strand_a_r_01!(ComponentPose=@n20, ExternalForce="0,0,0") -> (Pose) {
    alpha=0.7
    angularDampingOverride=0.999
    angularSpringConstant=7000
    bAngularSpring=true
    bOverrideAngularDamping=true
    bUseGravityOverride=true
    gravityScale=0.35
    simulationSpace=RootRelative
    boundBone {boneName=dyn_hair_strand_a_r_01}
    componentAppliedLinearAccClamp=<0 0 0>
    gravityOverride {z=-980}
    physicsBodyDefinitions [{
      boundBone {boneName=dyn_hair_strand_a_r_01}
      boxExtents=<15 20 20>
      constraintSetup {angularLimitsMax=<5 5 5>, angularLimitsMin=<-5 0 -5>, angularTarget=<1 0 0>}
      localJointOffset=<2 0 0>
     }]
    relativeSpaceBone {boneName=root}
    simSpaceSettings {maxAngularAcceleration=300, maxAngularVelocity=12, simSpaceAngularAlpha=0.4}
   }
   n22 = Anim Dynamics - Bone: dyn_hair_strand_a_r_02!(ComponentPose=@n21, ExternalForce="0,0,0") -> (Pose) {
    alpha=0.7
    angularDampingOverride=0.999
    angularSpringConstant=7000
    bAngularSpring=true
    bOverrideAngularDamping=true
    bUseGravityOverride=true
    gravityScale=0.35
    simulationSpace=RootRelative
    boundBone {boneName=dyn_hair_strand_a_r_02}
    componentAppliedLinearAccClamp=<0 0 0>
    gravityOverride {z=-980}
    physicsBodyDefinitions [{
      boundBone {boneName=dyn_hair_strand_a_r_02}
      boxExtents=<15 20 20>
      constraintSetup {angularLimitsMax=<10 10 10>, angularLimitsMin=<-10 0 -10>, angularTarget=<1 0 0>}
      localJointOffset=<2 0 0>
     }]
    relativeSpaceBone {boneName=root}
    simSpaceSettings {maxAngularAcceleration=300, maxAngularVelocity=12, simSpaceAngularAlpha=0.4}
   }
   n23 = Anim Dynamics - Bone: dyn_hair_strand_a_r_03!(ComponentPose=@n22, ExternalForce="0,0,0") -> (Pose) {
    alpha=0.7
    angularDampingOverride=0.999
    angularSpringConstant=7000
    bAngularSpring=true
    bOverrideAngularDamping=true
    bUseGravityOverride=true
    gravityScale=0.35
    simulationSpace=RootRelative
    boundBone {boneName=dyn_hair_strand_a_r_03}
    componentAppliedLinearAccClamp=<0 0 0>
    gravityOverride {z=-980}
    physicsBodyDefinitions [{
      boundBone {boneName=dyn_hair_strand_a_r_03}
      boxExtents=<15 20 20>
      constraintSetup {angularLimitsMax=<5 5 5>, angularLimitsMin=<-5 0 -5>, angularTarget=<1 0 0>}
      localJointOffset=<2 0 0>
     }]
    relativeSpaceBone {boneName=root}
    simSpaceSettings {maxAngularAcceleration=300, maxAngularVelocity=12, simSpaceAngularAlpha=0.4}
   }
   n24 = Anim Dynamics - Bone: dyn_hair_strand_b_r_01!(ComponentPose=@n23, ExternalForce="0,0,0") -> (Pose) {
    alpha=0.7
    angularDampingOverride=0.995
    angularSpringConstant=6000
    bAngularSpring=true
    bOverrideAngularDamping=true
    bUseGravityOverride=true
    gravityScale=0.35
    simulationSpace=RootRelative
    boundBone {boneName=dyn_hair_strand_b_r_01}
    componentAppliedLinearAccClamp=<0 0 0>
    gravityOverride {z=-980}
    physicsBodyDefinitions [{
      boundBone {boneName=dyn_hair_strand_b_r_01}
      boxExtents=<15 15 20>
      constraintSetup {angularLimitsMax=<5 5 5>, angularLimitsMin=<-5 0 -5>, angularTarget=<1 0 0>}
      localJointOffset=<2 0.5 0>
     }]
    relativeSpaceBone {boneName=root}
    simSpaceSettings {maxAngularAcceleration=300, maxAngularVelocity=12, simSpaceAngularAlpha=0.4}
   }
   n25 = Anim Dynamics - Bone: dyn_hair_strand_b_r_02!(ComponentPose=@n24, ExternalForce="0,0,0") -> (Pose) {
    alpha=0.7
    angularDampingOverride=0.995
    angularSpringConstant=6000
    bAngularSpring=true
    bOverrideAngularDamping=true
    bUseGravityOverride=true
    gravityScale=0.35
    simulationSpace=RootRelative
    boundBone {boneName=dyn_hair_strand_b_r_02}
    componentAppliedLinearAccClamp=<0 0 0>
    gravityOverride {z=-980}
    physicsBodyDefinitions [{
      boundBone {boneName=dyn_hair_strand_b_r_02}
      boxExtents=<15 15 20>
      constraintSetup {angularLimitsMax=<10 10 10>, angularLimitsMin=<-10 0 -10>, angularTarget=<1 0 0>}
      localJointOffset=<2 0.5 0>
     }]
    relativeSpaceBone {boneName=root}
    simSpaceSettings {maxAngularAcceleration=300, maxAngularVelocity=12, simSpaceAngularAlpha=0.4}
   }
   n26 = Anim Dynamics - Bone: dyn_hair_strand_b_r_03!(ComponentPose=@n25, ExternalForce="0,0,0") -> (Pose) {
    alpha=0.7
    angularDampingOverride=0.995
    angularSpringConstant=6000
    bAngularSpring=true
    bOverrideAngularDamping=true
    bUseGravityOverride=true
    gravityScale=0.35
    simulationSpace=RootRelative
    boundBone {boneName=dyn_hair_strand_b_r_03}
    componentAppliedLinearAccClamp=<0 0 0>
    gravityOverride {z=-980}
    physicsBodyDefinitions [{
      boundBone {boneName=dyn_hair_strand_b_r_03}
      boxExtents=<15 15 20>
      constraintSetup {angularLimitsMax=<15 15 15>, angularLimitsMin=<-15 0 -15>, angularTarget=<1 0 0>}
      localJointOffset=<2 0.5 0>
     }]
    relativeSpaceBone {boneName=root}
    simSpaceSettings {maxAngularAcceleration=300, maxAngularVelocity=12, simSpaceAngularAlpha=0.4}
   }
   n27 = Anim Dynamics - Bone: dyn_hair_strand_c_r_01!(ComponentPose=@n26, ExternalForce="0,0,0") -> (Pose) {
    alpha=0.7
    angularDampingOverride=0.9992
    angularSpringConstant=7500
    bAngularSpring=true
    bOverrideAngularDamping=true
    bUseGravityOverride=true
    gravityScale=0.35
    simulationSpace=RootRelative
    boundBone {boneName=dyn_hair_strand_c_r_01}
    componentAppliedLinearAccClamp=<0 0 0>
    gravityOverride {z=-980}
    physicsBodyDefinitions [{
      boundBone {boneName=dyn_hair_strand_c_r_01}
      boxExtents=<15 20 20>
      constraintSetup {angularLimitsMax=<5 5 5>, angularLimitsMin=<-5 0 -5>, angularTarget=<1 0 0>}
      localJointOffset=<3 0 0.5>
     }]
    relativeSpaceBone {boneName=root}
    simSpaceSettings {maxAngularAcceleration=300, maxAngularVelocity=12, simSpaceAngularAlpha=0.4}
   }
   n28 = Anim Dynamics - Bone: dyn_hair_strand_c_r_02!(ComponentPose=@n27, ExternalForce="0,0,0") -> (Pose) {
    alpha=0.7
    angularDampingOverride=0.9992
    angularSpringConstant=7500
    bAngularSpring=true
    bOverrideAngularDamping=true
    bUseGravityOverride=true
    gravityScale=0.35
    simulationSpace=RootRelative
    boundBone {boneName=dyn_hair_strand_c_r_02}
    componentAppliedLinearAccClamp=<0 0 0>
    gravityOverride {z=-980}
    physicsBodyDefinitions [{
      boundBone {boneName=dyn_hair_strand_c_r_02}
      boxExtents=<15 20 20>
      constraintSetup {angularLimitsMax=<10 10 10>, angularLimitsMin=<-10 0 -10>, angularTarget=<1 0 0>}
      localJointOffset=<3 0 0.5>
     }]
    relativeSpaceBone {boneName=root}
    simSpaceSettings {maxAngularAcceleration=300, maxAngularVelocity=12, simSpaceAngularAlpha=0.4}
   }
   n29 = Anim Dynamics - Bone: dyn_hair_strand_c_r_03!(ComponentPose=@n28, ExternalForce="0,0,0") -> (Pose) {
    alpha=0.7
    angularDampingOverride=0.9992
    angularSpringConstant=7500
    bAngularSpring=true
    bOverrideAngularDamping=true
    bUseGravityOverride=true
    gravityScale=0.35
    simulationSpace=RootRelative
    boundBone {boneName=dyn_hair_strand_c_r_03}
    componentAppliedLinearAccClamp=<0 0 0>
    gravityOverride {z=-980}
    physicsBodyDefinitions [{
      boundBone {boneName=dyn_hair_strand_c_r_03}
      boxExtents=<15 20 20>
      constraintSetup {angularLimitsMax=<15 15 15>, angularLimitsMin=<-15 0 -15>, angularTarget=<1 0 0>}
      localJointOffset=<3 0 0.5>
     }]
    relativeSpaceBone {boneName=root}
    simSpaceSettings {maxAngularAcceleration=300, maxAngularVelocity=12, simSpaceAngularAlpha=0.4}
   }
   n30 = Anim Dynamics - Bone: dyn_hair_strand_c_r_04!(ComponentPose=@n29, ExternalForce="0,0,0") -> (Pose) {
    alpha=0.7
    angularDampingOverride=0.9992
    angularSpringConstant=7500
    bAngularSpring=true
    bOverrideAngularDamping=true
    bUseGravityOverride=true
    gravityScale=0.35
    simulationSpace=RootRelative
    boundBone {boneName=dyn_hair_strand_c_r_04}
    componentAppliedLinearAccClamp=<0 0 0>
    gravityOverride {z=-980}
    physicsBodyDefinitions [{
      boundBone {boneName=dyn_hair_strand_c_r_04}
      boxExtents=<15 20 20>
      constraintSetup {angularLimitsMax=<25 25 25>, angularLimitsMin=<-25 0 -25>, angularTarget=<1 0 0>}
      localJointOffset=<3 0 0.5>
     }]
    relativeSpaceBone {boneName=root}
    simSpaceSettings {maxAngularAcceleration=300, maxAngularVelocity=12, simSpaceAngularAlpha=0.4}
   }
   n31 = Anim Dynamics - Bone: dyn_hair_strand_c_r_05!(ComponentPose=@n30, ExternalForce="0,0,0") -> (Pose) {
    alpha=0.7
    angularDampingOverride=0.9992
    angularSpringConstant=7500
    bAngularSpring=true
    bOverrideAngularDamping=true
    bUseGravityOverride=true
    gravityScale=0.35
    simulationSpace=RootRelative
    boundBone {boneName=dyn_hair_strand_c_r_05}
    componentAppliedLinearAccClamp=<0 0 0>
    gravityOverride {z=-980}
    physicsBodyDefinitions [{
      boundBone {boneName=dyn_hair_strand_c_r_05}
      boxExtents=<15 20 20>
      constraintSetup {angularLimitsMax=<15 15 15>, angularLimitsMin=<-15 0 -15>, angularTarget=<1 0 0>}
      localJointOffset=<3 0 0.5>
     }]
    relativeSpaceBone {boneName=root}
    simSpaceSettings {maxAngularAcceleration=300, maxAngularVelocity=12, simSpaceAngularAlpha=0.4}
   }
   n32 = Anim Dynamics - Bone: dyn_hair_strands_mid!(ComponentPose=@n31, ExternalForce="0,0,0") -> (Pose) {
    alpha=0
    angularDampingOverride=0.99
    angularSpringConstant=5000
    bAngularSpring=true
    bOverrideAngularDamping=true
    bUseGravityOverride=true
    gravityScale=0.35
    simulationSpace=RootRelative
    boundBone {boneName=dyn_hair_strands_mid}
    componentAppliedLinearAccClamp=<0 0 0>
    gravityOverride {z=-980}
    physicsBodyDefinitions [{
      boundBone {boneName=dyn_hair_strands_mid}
      boxExtents=<15 20 20>
      constraintSetup {angularLimitsMax=<5 5 5>, angularLimitsMin=<-5 0 -5>, angularTarget=<1 0 0>}
      localJointOffset=<2 0 0>
     }]
    relativeSpaceBone {boneName=root}
    simSpaceSettings {maxAngularAcceleration=300, maxAngularVelocity=12, simSpaceAngularAlpha=0.4}
   }
   n33 = Anim Dynamics - Bone: dyn_hair_strand_a_mid_01!(ComponentPose=@n32, ExternalForce="0,0,0") -> (Pose) {
    alpha=0.7
    angularDampingOverride=0.99
    angularSpringConstant=5000
    bAngularSpring=true
    bOverrideAngularDamping=true
    bUseGravityOverride=true
    gravityScale=0.35
    simulationSpace=RootRelative
    boundBone {boneName=dyn_hair_strand_a_mid_01}
    componentAppliedLinearAccClamp=<0 0 0>
    gravityOverride {z=-980}
    physicsBodyDefinitions [{
      boundBone {boneName=dyn_hair_strand_a_mid_01}
      boxExtents=<15 20 20>
      constraintSetup {angularLimitsMax=<5 5 5>, angularLimitsMin=<-5 0 -5>, angularTarget=<1 0 0>}
      localJointOffset=<2 0 0>
     }]
    relativeSpaceBone {boneName=root}
    simSpaceSettings {maxAngularAcceleration=300, maxAngularVelocity=12, simSpaceAngularAlpha=0.4}
   }
   n34 = Anim Dynamics - Bone: dyn_hair_strand_a_mid_02!(ComponentPose=@n33, ExternalForce="0,0,0") -> (Pose) {
    alpha=0.7
    angularDampingOverride=0.99
    angularSpringConstant=5000
    bAngularSpring=true
    bOverrideAngularDamping=true
    bUseGravityOverride=true
    gravityScale=0.35
    simulationSpace=RootRelative
    boundBone {boneName=dyn_hair_strand_a_mid_02}
    componentAppliedLinearAccClamp=<0 0 0>
    gravityOverride {z=-980}
    physicsBodyDefinitions [{
      boundBone {boneName=dyn_hair_strand_a_mid_02}
      boxExtents=<15 20 20>
      constraintSetup {angularLimitsMax=<10 10 10>, angularLimitsMin=<-10 0 -10>, angularTarget=<1 0 0>}
      localJointOffset=<2 0 0>
     }]
    relativeSpaceBone {boneName=root}
    simSpaceSettings {maxAngularAcceleration=300, maxAngularVelocity=12, simSpaceAngularAlpha=0.4}
   }
   n35 = Anim Dynamics - Bone: dyn_hair_strand_b_mid_01!(ComponentPose=@n34, ExternalForce="0,0,0") -> (Pose) {
    alpha=0.7
    angularDampingOverride=0.99
    angularSpringConstant=5000
    bAngularSpring=true
    bOverrideAngularDamping=true
    bUseGravityOverride=true
    gravityScale=0.35
    simulationSpace=RootRelative
    boundBone {boneName=dyn_hair_strand_b_mid_01}
    componentAppliedLinearAccClamp=<0 0 0>
    gravityOverride {z=-980}
    physicsBodyDefinitions [{
      boundBone {boneName=dyn_hair_strand_b_mid_01}
      boxExtents=<15 20 20>
      constraintSetup {angularLimitsMax=<5 5 5>, angularLimitsMin=<-5 0 -5>, angularTarget=<1 0 0>}
      localJointOffset=<2 0 0>
     }]
    relativeSpaceBone {boneName=root}
    simSpaceSettings {maxAngularAcceleration=300, maxAngularVelocity=12, simSpaceAngularAlpha=0.4}
   }
   n36 = Anim Dynamics - Bone: dyn_hair_strand_b_mid_02!(ComponentPose=@n35, ExternalForce="0,0,0") -> (Pose) {
    alpha=0.7
    angularDampingOverride=0.99
    angularSpringConstant=5000
    bAngularSpring=true
    bOverrideAngularDamping=true
    bUseGravityOverride=true
    gravityScale=0.35
    simulationSpace=RootRelative
    boundBone {boneName=dyn_hair_strand_b_mid_02}
    componentAppliedLinearAccClamp=<0 0 0>
    gravityOverride {z=-980}
    physicsBodyDefinitions [{
      boundBone {boneName=dyn_hair_strand_b_mid_02}
      boxExtents=<15 20 20>
      constraintSetup {angularLimitsMax=<10 10 10>, angularLimitsMin=<-10 0 -10>, angularTarget=<1 0 0>}
      localJointOffset=<2 0 0>
     }]
    relativeSpaceBone {boneName=root}
    simSpaceSettings {maxAngularAcceleration=300, maxAngularVelocity=12, simSpaceAngularAlpha=0.4}
   }
   n37 = Anim Dynamics - Bone: dyn_hair_strand_c_mid_01!(ComponentPose=@n36, ExternalForce="0,0,0") -> (Pose) {
    alpha=0.7
    angularDampingOverride=0.99
    angularSpringConstant=5000
    bAngularSpring=true
    bOverrideAngularDamping=true
    bUseGravityOverride=true
    gravityScale=0.35
    simulationSpace=RootRelative
    boundBone {boneName=dyn_hair_strand_c_mid_01}
    componentAppliedLinearAccClamp=<0 0 0>
    gravityOverride {z=-980}
    physicsBodyDefinitions [{
      boundBone {boneName=dyn_hair_strand_c_mid_01}
      boxExtents=<15 20 20>
      constraintSetup {angularLimitsMax=<5 5 5>, angularLimitsMin=<-5 0 -5>, angularTarget=<1 0 0>}
      localJointOffset=<2 0 0>
     }]
    relativeSpaceBone {boneName=root}
    simSpaceSettings {maxAngularAcceleration=300, maxAngularVelocity=12, simSpaceAngularAlpha=0.4}
   }
   n38 = Anim Dynamics - Bone: dyn_hair_strand_c_mid_02!(ComponentPose=@n37, ExternalForce="0,0,0") -> (Pose) {
    alpha=0.7
    angularDampingOverride=0.99
    angularSpringConstant=5000
    bAngularSpring=true
    bOverrideAngularDamping=true
    bUseGravityOverride=true
    gravityScale=0.35
    simulationSpace=RootRelative
    boundBone {boneName=dyn_hair_strand_c_mid_02}
    componentAppliedLinearAccClamp=<0 0 0>
    gravityOverride {z=-980}
    physicsBodyDefinitions [{
      boundBone {boneName=dyn_hair_strand_c_mid_02}
      boxExtents=<15 20 20>
      constraintSetup {angularLimitsMax=<10 10 10>, angularLimitsMin=<-10 0 -10>, angularTarget=<1 0 0>}
      localJointOffset=<2 0 0>
     }]
    relativeSpaceBone {boneName=root}
    simSpaceSettings {maxAngularAcceleration=300, maxAngularVelocity=12, simSpaceAngularAlpha=0.4}
   }
   n39 = Anim Dynamics - Bone: dyn_hair_strand_d_mid_01!(ComponentPose=@n38, ExternalForce="0,0,0") -> (Pose) {
    alpha=0.7
    angularDampingOverride=0.99
    angularSpringConstant=5000
    bAngularSpring=true
    bOverrideAngularDamping=true
    bUseGravityOverride=true
    gravityScale=0.35
    simulationSpace=RootRelative
    boundBone {boneName=dyn_hair_strand_d_mid_01}
    componentAppliedLinearAccClamp=<0 0 0>
    gravityOverride {z=-980}
    physicsBodyDefinitions [{
      boundBone {boneName=dyn_hair_strand_d_mid_01}
      boxExtents=<15 20 20>
      constraintSetup {angularLimitsMax=<5 5 5>, angularLimitsMin=<-5 0 -5>, angularTarget=<1 0 0>}
      localJointOffset=<2 0 0>
     }]
    relativeSpaceBone {boneName=root}
    simSpaceSettings {maxAngularAcceleration=300, maxAngularVelocity=12, simSpaceAngularAlpha=0.4}
   }
   n40 = Anim Dynamics - Bone: dyn_hair_strand_d_mid_02!(ComponentPose=@n39, ExternalForce="0,0,0") -> (Pose) {
    alpha=0.7
    angularDampingOverride=0.99
    angularSpringConstant=5000
    bAngularSpring=true
    bOverrideAngularDamping=true
    bUseGravityOverride=true
    gravityScale=0.35
    simulationSpace=RootRelative
    boundBone {boneName=dyn_hair_strand_d_mid_02}
    componentAppliedLinearAccClamp=<0 0 0>
    gravityOverride {z=-980}
    physicsBodyDefinitions [{
      boundBone {boneName=dyn_hair_strand_d_mid_02}
      boxExtents=<15 20 20>
      constraintSetup {angularLimitsMax=<10 10 15>, angularLimitsMin=<-10 0 -15>, angularTarget=<1 0 0>}
      localJointOffset=<2 0 0>
     }]
    relativeSpaceBone {boneName=root}
    simSpaceSettings {maxAngularAcceleration=300, maxAngularVelocity=12, simSpaceAngularAlpha=0.4}
   }
   n41 = Anim Dynamics - Bone: dyn_hair_strand_d_mid_03!(ComponentPose=@n40, ExternalForce="0,0,0") -> (Pose) {
    alpha=0.7
    angularDampingOverride=0.99
    angularSpringConstant=5000
    bAngularSpring=true
    bOverrideAngularDamping=true
    bUseGravityOverride=true
    gravityScale=0.35
    simulationSpace=RootRelative
    boundBone {boneName=dyn_hair_strand_d_mid_03}
    componentAppliedLinearAccClamp=<0 0 0>
    gravityOverride {z=-980}
    physicsBodyDefinitions [{
      boundBone {boneName=dyn_hair_strand_d_mid_03}
      boxExtents=<15 20 20>
      constraintSetup {angularLimitsMax=<15 15 15>, angularLimitsMin=<-15 0 -15>, angularTarget=<1 0 0>}
      localJointOffset=<2 0 0>
     }]
    relativeSpaceBone {boneName=root}
    simSpaceSettings {maxAngularAcceleration=300, maxAngularVelocity=12, simSpaceAngularAlpha=0.4}
   }
   Output Pose(Result=Component To Local(ComponentPose=@n41)) {layerGroup=None}
  ]}
 EventGraph {nodes [Event BlueprintUpdateAnimation()]}
]
variables []
